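# Control Rig for the female QuicheLorraineCrisp character body, driving corrective and secondary bones for jacket, shoulders, armpits, cuffs and feet.
# Spherical pose readers on thighs, calf bend and pelvis height set alpha variables in a sequenced pass.
# Parent constraints tie jacket bones to the thighs, then weighted ModifyTransforms layer corrective poses, including smoothed front-end and emote-specific fixes.
# Upper-arm twist extraction and pose readers drive shoulder and armpit deformation.
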ControlRigBlueprint /BRCosmetics/Characters/Player/Female/Medium/Bodies/F_MED_QuicheLorraineCrisp/Meshes/F_MED_QuicheLorraineCrisp_CtrlRig
graphs [
 Multiply_4_ContainedGraph {nodes [
   n1 = RigVMFunctionEntryNode()
   n2 = &RigVMFunction_MathDoubleMul(A=@n1.A, B=@n1.B)
   RigVMFunctionReturnNode(Result=&RigVMFunction_MathDoubleMul(B=@n1.D, A=&RigVMFunction_MathDoubleMul(B=@n1.C, A=@n2.Result).Result).Result)
  ]}
 RigVMFunctionLibrary {links [], nodes []}
 RigVMFunction_Sequence_1_ContainedGraph {nodes [
   n1 = RigVMFunctionEntryNode()
   RigVMFunction_Sequence = &RigVMFunction_Sequence(ExecuteContext=@n1.ExecuteContext)
   n2 = &RigVMFunction_Sequence(ExecuteContext=@RigVMFunction_Sequence.B)
   n3 = &RigVMFunction_Sequence(ExecuteContext=@n2.B)
   n4 = &RigVMFunction_Sequence(ExecuteContext=@n3.B)
   n5 = &RigVMFunction_Sequence(ExecuteContext=@n4.B)
   n6 = &RigVMFunction_Sequence(ExecuteContext=@n5.B)
   RigVMFunctionReturnNode(A=@RigVMFunction_Sequence.A, B=@n2.A, C=@n3.A, D=@n4.A, E=@n5.A, F=@n6.A, G=@n6.B)
  ]}
 RigVMModel {nodes [
   RigUnit_BeginExecution = &RigUnit_BeginExecution()
   RigVMFunction_Sequence = RigVMAggregateNode(ExecuteContext=@RigUnit_BeginExecution.ExecutePin)
   &RigUnit_ParentConstraint(
    AdvancedSettings=(InterpolationType=Average,RotationOrderForFilter=XZY)
    Child=(Type=Bone,Name="dyn_jacket_fr_l_base")
    Parents=((Item=(Type=Bone,Name="thigh_l"),Weight=1))
    Weight=1
    ExecutePin=@RigVMFunction_Sequence.C
   )
   > &RigUnit_ParentConstraint(
    AdvancedSettings=(InterpolationType=Average,RotationOrderForFilter=XZY)
    Child=(Type=Bone,Name="dyn_jacket_fr_r_base")
    Parents=((Item=(Type=Bone,Name="thigh_r"),Weight=1))
    Weight=0.8
   )
   > &RigUnit_ParentConstraint(
    AdvancedSettings=(InterpolationType=Average,RotationOrderForFilter=XZY)
    Child=(Type=Bone,Name="dyn_jacket_side_l_base")
    Parents=((Item=(Type=Bone,Name="dyn_jacket_fr_l_base"),Weight=1))
    Weight=1
   )
   > &RigUnit_ParentConstraint(
    AdvancedSettings=(InterpolationType=Average,RotationOrderForFilter=XZY)
    Child=(Type=Bone,Name="dyn_jacket_side_r_base")
    Parents=((Item=(Type=Bone,Name="dyn_jacket_fr_r_base"),Weight=1))
    Weight=1
   )
   > &RigUnit_ParentConstraint(
    AdvancedSettings=(InterpolationType=Average,RotationOrderForFilter=XZY)
    Child=(Type=Bone,Name="dyn_jacket_bck_l_base")
    Parents=((Item=(Type=Bone,Name="dyn_jacket_side_l_base"),Weight=0.5),(Item=(Type=Bone,Name="dyn_jacket_bck_c_base"),Weight=0.5))
    Weight=1
   )
   > &RigUnit_ParentConstraint(
    AdvancedSettings=(InterpolationType=Average,RotationOrderForFilter=XZY)
    Child=(Type=Bone,Name="dyn_jacket_bck_r_base")
    Parents=((Item=(Type=Bone,Name="dyn_jacket_side_r_base"),Weight=0.5),(Item=(Type=Bone,Name="dyn_jacket_bck_c_base"),Weight=0.5))
    Weight=1
   )
   > &RigUnit_ParentConstraint(
    AdvancedSettings=(InterpolationType=Average,RotationOrderForFilter=XZY)
    Child=(Type=Bone,Name="dyn_jacket_bck_c_base")
    Parents=((Item=(Type=Bone,Name="dyn_jacket_bck_l_base"),Weight=0.5),(Item=(Type=Bone,Name="dyn_jacket_bck_r_base"),Weight=0.5))
    Weight=1
   )
   n1 = &RigVMFunction_AlphaInterp(InterpSpeedDecreasing=30, InterpSpeedIncreasing=30, Scale=1, bClampResult=true, Value=RigVMVariableNode() {variable=FrontEnd Alpha}.Value)
   > &RigUnit_ModifyTransforms(
    ItemToModify="((Item=(Type=Bone,Name=\"dyn_jacket_fr_r_base\"),Transform=(Rotation=(X=0,Y=-0.0872,Z=0,W=0.9962),Translation=(X=0,Y=0,Z=0),Scale3D=(X=1,Y=1,Z=1))),(Item=(Type=Bone,Name=\"dyn_jacket_side_r_base\"),Transform=(Rotation=(X=0,Y=-0.0872,Z=0,W=0.9962),Translation=( ...(truncated)"
    WeightMinimum=0
    Weight=@n1.Result
   )
   > &RigUnit_ModifyTransforms(
    ItemToModify="((Item=(Type=Bone,Name=\"deform_armpit_l\"),Transform=(Rotation=(X=0,Y=0,Z=0,W=1),Translation=(X=1.2,Y=0,Z=0),Scale3D=(X=1,Y=1,Z=1))),(Item=(Type=Bone,Name=\"deform_armpit_r\"),Transform=(Rotation=(X=0,Y=0,Z=0,W=1),Translation=(X=1.2,Y=0,Z=0),Scale3D=(X=1,Y=1, ...(truncated)"
    WeightMinimum=0
    Weight=@n1.Result
   )
   > &RigUnit_ModifyTransforms(
    ItemToModify="((Item=(Type=Bone,Name=\"dyn_jacket_fr_l_base\"),Transform=(Rotation=(X=0.1768,Y=-0.2996,Z=0.548,W=-0.7607),Translation=(X=10.9712,Y=-0.2648,Z=-5.6416),Scale3D=(X=1,Y=1,Z=1))),(Item=(Type=Bone,Name=\"dyn_jacket_side_l_base\"),Transform=(Rotation=(X=0,Y=0.4226, ...(truncated)"
    WeightMinimum=0
    Weight=RigVMVariableNode() {variable=L thighForward Alpha}.Value
   )
   > &RigUnit_ModifyTransforms(
    ItemToModify="((Item=(Type=Bone,Name=\"dyn_jacket_fr_r_base\"),Transform=(Rotation=(X=0.1638,Y=-0.0289,Z=-0.363,W=-0.9168),Translation=(X=4.3419,Y=3.2724,Z=3.6433),Scale3D=(X=1,Y=1,Z=1))),(Item=(Type=Bone,Name=\"dyn_jacket_side_r_base\"),Transform=(Rotation=(X=0.0298,Y=0.08 ...(truncated)"
    WeightMinimum=0
    Weight=RigVMVariableNode() {variable=R thighForward Alpha}.Value
   )
   > &RigUnit_ModifyTransforms(
    ItemToModify="((Item=(Type=Bone,Name=\"dyn_jacket_fr_l_base\"),Transform=(Rotation=(X=-0.0076,Y=-0.0868,Z=0.2578,W=-0.9623),Translation=(X=0,Y=0,Z=0),Scale3D=(X=1,Y=1,Z=1))),(Item=(Type=Bone,Name=\"dyn_jacket_side_l_base\"),Transform=(Rotation=(X=0,Y=0,Z=0.2588,W=-0.9659),T ...(truncated)"
    WeightMinimum=0
    Weight=RigVMVariableNode() {variable=L thighSide Alpha}.Value
   )
   > &RigUnit_ModifyTransforms(
    ItemToModify="((Item=(Type=Bone,Name=\"dyn_jacket_fr_r_base\"),Transform=(Rotation=(X=0,Y=0,Z=0,W=1),Translation=(X=0,Y=0,Z=0),Scale3D=(X=1,Y=1,Z=1))),(Item=(Type=Bone,Name=\"dyn_jacket_side_r_base\"),Transform=(Rotation=(X=0,Y=0,Z=0,W=1),Translation=(X=0,Y=0,Z=0),Scale3D=( ...(truncated)"
    WeightMinimum=0
    Weight=RigVMVariableNode() {variable=R thighSide Alpha}.Value
   )
   > &RigUnit_ModifyTransforms(
    ItemToModify="((Item=(Type=Bone,Name=\"dyn_jacket_bck_c_base\"),Transform=(Rotation=(X=0,Y=0.342,Z=0,W=-0.9397),Translation=(X=-16.8216,Y=0.0004,Z=-2.0463),Scale3D=(X=1,Y=1,Z=1))),(Item=(Type=Bone,Name=\"dyn_jacket_bck_r_base\"),Transform=(Rotation=(X=-0.0151,Y=0.0858,Z=-0. ...(truncated)"
    WeightMinimum=0
    Weight=RigVMVariableNode() {variable=Low 2 Ground Alpha}.Value
   )
   n2 = &RigUnit_SphericalPoseReader(
    ActiveRegionScaleFactors=(PositiveWidth=4,NegativeWidth=3,PositiveHeight=2.5,NegativeHeight=2)
    ActiveRegionSize=0.1
    Debug=(bDrawDebug=false,bDraw2D=False,bDrawLocalAxes=False,DebugScale=25,DebugSegments=20,DebugThickness=0.25)
    DriverAxis=(X=-1,Y=0,Z=0)
    DriverItem=(Type=Bone,Name="thigh_l")
    FalloffRegionScaleFactors=(PositiveWidth=1,NegativeWidth=3,PositiveHeight=1,NegativeHeight=0.5)
    FalloffSize=0.4
    FlipHeightScaling=False
    FlipWidthScaling=False
    OptionalParentItem=(Type=Bone,Name="None")
    RotationOffset=(X=100,Y=0,Z=-20)
    ExecutePin=@RigVMFunction_Sequence.A
   )
   n3 = RigVMVariableNode(Value=@n2.OutputParam, ExecuteContext=@RigVMFunction_Sequence.B) {variable=L thighForward Alpha}
   n4 = &RigUnit_SphericalPoseReader(
    ActiveRegionScaleFactors=(PositiveWidth=4,NegativeWidth=3,PositiveHeight=2.5,NegativeHeight=2)
    ActiveRegionSize=0.1
    Debug=(bDrawDebug=false,bDraw2D=False,bDrawLocalAxes=False,DebugScale=25,DebugSegments=20,DebugThickness=0.25)
    DriverAxis=(X=1,Y=0,Z=0)
    DriverItem=(Type=Bone,Name="thigh_r")
    FalloffRegionScaleFactors=(PositiveWidth=1,NegativeWidth=3.6,PositiveHeight=1,NegativeHeight=0.5)
    FalloffSize=0.4
    FlipHeightScaling=False
    FlipWidthScaling=true
    OptionalParentItem=(Type=Bone,Name="None")
    RotationOffset=(X=-80,Y=0,Z=-35.3405)
    ExecutePin=@n2.ExecutePin
   )
   n5 = RigVMVariableNode(Value=@n4.OutputParam, ExecuteContext=@n3.ExecuteContext) {variable=R thighForward Alpha}
   n6 = &RigUnit_SphericalPoseReader(
    ActiveRegionScaleFactors=(PositiveWidth=3,NegativeWidth=2,PositiveHeight=2,NegativeHeight=2)
    ActiveRegionSize=0.1
    Debug=(bDrawDebug=false,bDraw2D=False,bDrawLocalAxes=False,DebugScale=25,DebugSegments=20,DebugThickness=0.25)
    DriverAxis=(X=-1,Y=0,Z=0)
    DriverItem=(Type=Bone,Name="thigh_l")
    FalloffRegionScaleFactors=(PositiveWidth=2,NegativeWidth=2,PositiveHeight=2,NegativeHeight=2)
    FalloffSize=0.25
    FlipHeightScaling=false
    FlipWidthScaling=false
    OptionalParentItem=(Type=Bone,Name="None")
    RotationOffset=(X=180,Y=-40,Z=0)
    ExecutePin=@n4.ExecutePin
   )
   n7 = RigVMVariableNode(Value=@n6.OutputParam, ExecuteContext=@n5.ExecuteContext) {variable=L thighSide Alpha}
   n8 = &RigUnit_SphericalPoseReader(
    ActiveRegionScaleFactors=(PositiveWidth=3,NegativeWidth=2,PositiveHeight=2,NegativeHeight=2)
    ActiveRegionSize=0.1
    Debug=(bDrawDebug=false,bDraw2D=False,bDrawLocalAxes=False,DebugScale=25,DebugSegments=20,DebugThickness=0.25)
    DriverAxis=(X=1,Y=0,Z=0)
    DriverItem=(Type=Bone,Name="thigh_r")
    FalloffRegionScaleFactors=(PositiveWidth=2,NegativeWidth=2,PositiveHeight=2,NegativeHeight=2)
    FalloffSize=0.25
    FlipHeightScaling=False
    FlipWidthScaling=true
    OptionalParentItem=(Type=Bone,Name="None")
    RotationOffset=(X=0,Y=-40,Z=0)
    ExecutePin=@n6.ExecutePin
   )
   n9 = RigVMVariableNode(Value=@n8.OutputParam, ExecuteContext=@n7.ExecuteContext) {variable=R thighSide Alpha}
   n10 = &RigUnit_SphericalPoseReader(
    ActiveRegionScaleFactors=(PositiveWidth=2,NegativeWidth=2,PositiveHeight=2,NegativeHeight=1)
    ActiveRegionSize=0.1
    Debug=(bDrawDebug=false,bDraw2D=False,bDrawLocalAxes=False,DebugScale=25,DebugSegments=20,DebugThickness=0.25)
    DriverAxis=(X=-1,Y=0,Z=0)
    DriverItem=(Type=Bone,Name="thigh_l")
    FalloffRegionScaleFactors=(PositiveWidth=2,NegativeWidth=2,PositiveHeight=4,NegativeHeight=1)
    FalloffSize=0.3
    FlipHeightScaling=false
    FlipWidthScaling=false
    OptionalParentItem=(Type=Bone,Name="None")
    RotationOffset=(X=0,Y=90,Z=80)
    ExecutePin=@n8.ExecutePin
   )
   n11 = RigVMVariableNode(Value=@n10.OutputParam, ExecuteContext=@n9.ExecuteContext) {variable=L thighBack Alpha}
   n12 = &RigUnit_SphericalPoseReader(
    ActiveRegionScaleFactors=(PositiveWidth=2,NegativeWidth=2,PositiveHeight=2,NegativeHeight=1)
    ActiveRegionSize=0.1
    Debug=(bDrawDebug=false,bDraw2D=False,bDrawLocalAxes=False,DebugScale=25,DebugSegments=20,DebugThickness=0.25)
    DriverAxis=(X=1,Y=0,Z=0)
    DriverItem=(Type=Bone,Name="thigh_r")
    FalloffRegionScaleFactors=(PositiveWidth=2,NegativeWidth=2,PositiveHeight=3,NegativeHeight=1)
    FalloffSize=0.3
    FlipHeightScaling=false
    FlipWidthScaling=false
    OptionalParentItem=(Type=Bone,Name="None")
    RotationOffset=(X=180,Y=93.7803,Z=60.6503)
    ExecutePin=@n10.ExecutePin
   )
   n13 = RigVMVariableNode(Value=@n12.OutputParam, ExecuteContext=@n11.ExecuteContext) {variable=R thighBack Alpha}
   n14 = &RigUnit_SphericalPoseReader(
    ActiveRegionScaleFactors=(PositiveWidth=3,NegativeWidth=2,PositiveHeight=2,NegativeHeight=2)
    ActiveRegionSize=0.1
    Debug=(bDrawDebug=false,bDraw2D=False,bDrawLocalAxes=False,DebugScale=25,DebugSegments=20,DebugThickness=0.25)
    DriverAxis=(X=-1,Y=0,Z=0)
    DriverItem=(Type=Bone,Name="thigh_l")
    FalloffRegionScaleFactors=(PositiveWidth=2,NegativeWidth=2,PositiveHeight=2,NegativeHeight=2)
    FalloffSize=0.25
    FlipHeightScaling=false
    FlipWidthScaling=false
    OptionalParentItem=(Type=Bone,Name="None")
    RotationOffset=(X=2.688,Y=38.2586,Z=-2.2671)
    ExecutePin=@n12.ExecutePin
   )
   n15 = RigVMVariableNode(Value=@n14.OutputParam, ExecuteContext=@n13.ExecuteContext) {variable=L thighSideIn Alpha}
   n16 = &RigUnit_SphericalPoseReader(
    ActiveRegionScaleFactors=(PositiveWidth=3,NegativeWidth=2,PositiveHeight=2,NegativeHeight=2)
    ActiveRegionSize=0.1
    Debug=(bDrawDebug=false,bDraw2D=False,bDrawLocalAxes=False,DebugScale=25,DebugSegments=20,DebugThickness=0.25)
    DriverAxis=(X=1,Y=0,Z=0)
    DriverItem=(Type=Bone,Name="thigh_r")
    FalloffRegionScaleFactors=(PositiveWidth=2,NegativeWidth=2,PositiveHeight=2,NegativeHeight=2)
    FalloffSize=0.25
    FlipHeightScaling=False
    FlipWidthScaling=true
    OptionalParentItem=(Type=Bone,Name="None")
    RotationOffset=(X=0,Y=-40,Z=0)
    ExecutePin=@n14.ExecutePin
   )
   n17 = RigVMVariableNode(Value=@n16.OutputParam, ExecuteContext=@n15.ExecuteContext) {variable=R thighSideIn Alpha}
   n18 = &RigUnit_SphericalPoseReader(
    ActiveRegionScaleFactors=(PositiveWidth=1.5,NegativeWidth=1.5,PositiveHeight=1.5,NegativeHeight=1.5)
    ActiveRegionSize=0.1
    Debug=(bDrawDebug=true,bDraw2D=False,bDrawLocalAxes=False,DebugScale=25,DebugSegments=20,DebugThickness=0.25)
    DriverAxis=(X=-1,Y=0,Z=0)
    DriverItem=(Type=Bone,Name="thigh_l")
    FalloffRegionScaleFactors=(PositiveWidth=2,NegativeWidth=2,PositiveHeight=2,NegativeHeight=2)
    FalloffSize=0.2
    FlipHeightScaling=false
    FlipWidthScaling=false
    OptionalParentItem=(Type=Bone,Name="None")
    RotationOffset=(X=155.857,Y=-40,Z=-11.0727)
    ExecutePin=@n16.ExecutePin
   )
   n19 = RigVMVariableNode(Value=@n18.OutputParam, ExecuteContext=@n17.ExecuteContext) {variable=L thighSide Extreme Alpha}
   n20 = &RigUnit_SphericalPoseReader(
    ActiveRegionScaleFactors=(PositiveWidth=1.5,NegativeWidth=1.5,PositiveHeight=1.5,NegativeHeight=1.5)
    ActiveRegionSize=0.1
    Debug=(bDrawDebug=true,bDraw2D=False,bDrawLocalAxes=False,DebugScale=25,DebugSegments=20,DebugThickness=0.25)
    DriverAxis=(X=1,Y=0,Z=0)
    DriverItem=(Type=Bone,Name="thigh_r")
    FalloffRegionScaleFactors=(PositiveWidth=2,NegativeWidth=2,PositiveHeight=2,NegativeHeight=2)
    FalloffSize=0.2
    FlipHeightScaling=False
    FlipWidthScaling=true
    OptionalParentItem=(Type=Bone,Name="None")
    RotationOffset=(X=-27.7699,Y=-33.7334,Z=-9.1032)
    ExecutePin=@n18.ExecutePin
   )
   RigVMVariableNode(Value=@n20.OutputParam, ExecuteContext=@n19.ExecuteContext) {variable=R thighSide Extreme Alpha}
   &RigUnit_ParentConstraint(
    AdvancedSettings=(InterpolationType=Average,RotationOrderForFilter=XZY)
    Child=(Type=Bone,Name="deform_armpit_l_base")
    Parents=((Item=(Type=Bone,Name="spine_03"),Weight=0.3),(Item=(Type=Bone,Name="spine_04"),Weight=0.7))
    Weight=1
    ExecutePin=@RigVMFunction_Sequence.D
   )
   > &RigUnit_ParentConstraint(
    AdvancedSettings=(InterpolationType=Average,RotationOrderForFilter=XZY)
    Child=(Type=Bone,Name="deform_armpit_r_base")
    Parents=((Item=(Type=Bone,Name="spine_03"),Weight=0.3),(Item=(Type=Bone,Name="spine_04"),Weight=0.7))
    Weight=1
   )
   RigVMFunction_MathQuaternionSwingTwist = &RigVMFunction_MathQuaternionSwingTwist(
    TwistAxis=(X=1,Y=0,Z=0)
    Input=&RigUnit_GetTransform(Item=(Type=Bone,Name="upperarm_l"), Space=LocalSpace, bInitial=False).Transform.Rotation
   )
   Inverse = /Script/RigVM.RigVMFunction_MathQuaternionInverse(Value=@RigVMFunction_MathQuaternionSwingTwist.Twist) {is=struct RigVMUnitNode, template="Inverse::Execute(in Value,out Result)"}
   > Set Transform = &RigUnit_SetRotation(Item=(Type=Bone,Name="deform_shoulder_l"), Weight=0.8, bInitial=false, Value=@Inverse.Result)
   n21 = &RigVMFunction_MathQuaternionSwingTwist(
    TwistAxis=(X=1,Y=0,Z=0)
    Input=&RigUnit_GetTransform(Item=(Type=Bone,Name="upperarm_r"), Space=LocalSpace, bInitial=False).Transform.Rotation
   )
   > Set Transform_2 = &RigUnit_SetRotation(Item=(Type=Bone,Name="deform_shoulder_r"), Weight=0.8, bInitial=false, Value=@n21.Twist)
   > n22 = &RigUnit_SphericalPoseReader(
    ActiveRegionScaleFactors=(PositiveWidth=0.1,NegativeWidth=0.1,PositiveHeight=0.1,NegativeHeight=0.1)
    ActiveRegionSize=0.1
    Debug=(bDrawDebug=false,bDraw2D=false,DebugScale=25,DebugSegments=20,DebugThickness=0.25)
    DriverAxis=(X=1,Y=0,Z=0)
    DriverItem=(Type=Bone,Name="upperarm_l")
    FalloffRegionScaleFactors=(PositiveWidth=1,NegativeWidth=1,PositiveHeight=1,NegativeHeight=1)
    FalloffSize=0.45
    OptionalParentItem=(Type=Bone,Name="spine_05")
    RotationOffset=(X=0,Y=0,Z=0)
   )
   > &RigUnit_ModifyTransforms(
    ItemToModify="((Item=(Type=Bone,Name=\"deform_shoulder_l\"),Transform=(Rotation=(X=0,Y=0,Z=0,W=1),Translation=(X=0,Y=0,Z=0),Scale3D=(X=1,Y=1,Z=1))),(Item=(Type=Bone,Name=\"deform_armpit_l\"),Transform=(Rotation=(X=0,Y=0,Z=0,W=-1),Translation=(X=0,Y=0,Z=6.0186),Scale3D=(X=1, ...(truncated)"
    WeightMinimum=0
    Weight=@n22.OutputParam
   )
   > n23 = &RigUnit_SphericalPoseReader(
    ActiveRegionScaleFactors=(PositiveWidth=0.1,NegativeWidth=0.1,PositiveHeight=0.1,NegativeHeight=0.1)
    ActiveRegionSize=0.1
    Debug=(bDrawDebug=false,bDraw2D=false,DebugScale=25,DebugSegments=20,DebugThickness=0.25)
    DriverAxis=(X=-1,Y=0,Z=0)
    DriverItem=(Type=Bone,Name="upperarm_r")
    FalloffRegionScaleFactors=(PositiveWidth=1,NegativeWidth=1,PositiveHeight=1,NegativeHeight=1)
    FalloffSize=0.45
    OptionalParentItem=(Type=Bone,Name="spine_05")
    RotationOffset=(X=180,Y=0,Z=0)
   )
   > &RigUnit_ModifyTransforms(
    ItemToModify=((Item=(Type=Bone,Name="deform_shoulder_r"),Transform=(Rotation=(X=0,Y=0,Z=0,W=1),Translation=(X=0,Y=0,Z=0),Scale3D=(X=1,Y=1,Z=1))))
    WeightMinimum=0
    Weight=@n23.OutputParam
   )
   > n24 = &RigUnit_SphericalPoseReader(
    ActiveRegionScaleFactors=(PositiveWidth=0.25,NegativeWidth=0.25,PositiveHeight=0.25,NegativeHeight=0.25)
    ActiveRegionSize=0.1
    Debug=(bDrawDebug=false,bDraw2D=false,DebugScale=25,DebugSegments=20,DebugThickness=0.25)
    DriverAxis=(X=-1,Y=0,Z=0)
    DriverItem=(Type=Bone,Name="upperarm_r")
    FalloffRegionScaleFactors=(PositiveWidth=1,NegativeWidth=1,PositiveHeight=1,NegativeHeight=1)
    FalloffSize=0.45
    OptionalParentItem=(Type=Bone,Name="spine_05")
    RotationOffset=(X=-100,Y=0,Z=45)
   )
   > &RigUnit_ModifyTransforms(
    ItemToModify="((Item=(Type=Bone,Name=\"deform_shoulder_r\"),Transform=(Rotation=(X=0,Y=0.0872,Z=0,W=0.9962),Translation=(X=-3.6479,Y=-0.3991,Z=0.2576),Scale3D=(X=1,Y=1,Z=1))),(Item=(Type=Bone,Name=\"deform_coat_lapel_r_mid_base\"),Transform=(Rotation=(X=-0.1402,Y=-0.1417,Z= ...(truncated)"
    WeightMinimum=0
    Weight=@n24.OutputParam
   )
   > n25 = &RigUnit_SphericalPoseReader(
    ActiveRegionScaleFactors=(PositiveWidth=1,NegativeWidth=1,PositiveHeight=1,NegativeHeight=1)
    ActiveRegionSize=0.12
    Debug=(bDrawDebug=false,bDraw2D=false,DebugScale=25,DebugSegments=20,DebugThickness=0.25)
    DriverAxis=(X=-1,Y=0,Z=0)
    DriverItem=(Type=Bone,Name="upperarm_r")
    FalloffRegionScaleFactors=(PositiveWidth=1,NegativeWidth=1,PositiveHeight=1,NegativeHeight=1)
    FalloffSize=0.3
    OptionalParentItem=(Type=Bone,Name="spine_05")
    RotationOffset=(X=-271.8951,Y=19.3462,Z=34.7794)
   )
   > &RigUnit_ModifyTransforms(
    ItemToModify="((Item=(Type=Bone,Name=\"deform_shoulder_r\"),Transform=(Rotation=(X=0,Y=0,Z=0,W=-1),Translation=(X=-4.0561,Y=0.1767,Z=0),Scale3D=(X=1.2,Y=1.2,Z=1.2))),(Item=(Type=Bone,Name=\"deform_armpit_r\"),Transform=(Rotation=(X=0,Y=0,Z=0,W=1),Translation=(X=0,Y=0,Z=0),S ...(truncated)"
    WeightMinimum=0
    Weight=@n25.OutputParam
   )
   > n26 = &RigUnit_SphericalPoseReader(
    ActiveRegionScaleFactors=(PositiveWidth=1,NegativeWidth=1,PositiveHeight=1,NegativeHeight=1)
    ActiveRegionSize=0.12
    Debug=(bDrawDebug=false,bDraw2D=false,DebugScale=25,DebugSegments=20,DebugThickness=0.25)
    DriverAxis=(X=1,Y=0,Z=0)
    DriverItem=(Type=Bone,Name="upperarm_l")
    FalloffRegionScaleFactors=(PositiveWidth=1,NegativeWidth=1,PositiveHeight=1,NegativeHeight=1)
    FalloffSize=0.3
    OptionalParentItem=(Type=Bone,Name="spine_05")
    RotationOffset=(X=-120.1521,Y=70.5782,Z=0)
   )
   > &RigUnit_ModifyTransforms(
    ItemToModify="((Item=(Type=Bone,Name=\"deform_shoulder_l\"),Transform=(Rotation=(X=0,Y=0,Z=0,W=1),Translation=(X=-4.9336,Y=-0.1551,Z=-1.6639),Scale3D=(X=1.2,Y=1.2,Z=1.2))),(Item=(Type=Bone,Name=\"deform_armpit_l\"),Transform=(Rotation=(X=0,Y=0,Z=0,W=1),Translation=(X=0,Y=-4 ...(truncated)"
    WeightMinimum=0
    Weight=@n26.OutputParam
   )
   > n27 = &RigUnit_SphericalPoseReader(
    ActiveRegionScaleFactors=(PositiveWidth=0.5,NegativeWidth=0.5,PositiveHeight=0.5,NegativeHeight=0.5)
    ActiveRegionSize=0.1
    Debug=(bDrawDebug=false,bDraw2D=false,DebugScale=25,DebugSegments=20,DebugThickness=0.25)
    DriverAxis=(X=-1,Y=0,Z=0)
    DriverItem=(Type=Bone,Name="upperarm_r")
    FalloffRegionScaleFactors=(PositiveWidth=1,NegativeWidth=1,PositiveHeight=1,NegativeHeight=1)
    FalloffSize=0.3
    OptionalParentItem=(Type=Bone,Name="spine_05")
    RotationOffset=(X=-45.7091,Y=17.2417,Z=0)
   )
   > &RigUnit_ModifyTransforms(
    ItemToModify="((Item=(Type=Bone,Name=\"deform_shoulder_r\"),Transform=(Rotation=(X=0,Y=0,Z=0,W=1),Translation=(X=-3.3165,Y=1.0624,Z=0),Scale3D=(X=1,Y=1,Z=1))),(Item=(Type=Bone,Name=\"deform_armpit_r\"),Transform=(Rotation=(X=0,Y=0,Z=0,W=1),Translation=(X=-0.8137,Y=0,Z=2.498 ...(truncated)"
    WeightMinimum=0
    Weight=@n27.OutputParam
   )
   &RigUnit_ParentConstraint(
    AdvancedSettings=(InterpolationType=Average,RotationOrderForFilter=XZY)
    Child=(Type=Bone,Name="deform_ball_l")
    Filter=(TranslationFilter=(bX=false,bY=false,bZ=false),RotationFilter=(bX=True,bY=True,bZ=True),ScaleFilter=(bX=false,bY=false,bZ=false))
    Parents=((Item=(Type=Bone,Name="ball_l"),Weight=1))
    Weight=1
    ExecutePin=@RigVMFunction_Sequence.E
   )
   > &RigUnit_ParentConstraint(
    AdvancedSettings=(InterpolationType=Average,RotationOrderForFilter=XZY)
    Child=(Type=Bone,Name="deform_ball_r")
    Filter=(TranslationFilter=(bX=false,bY=false,bZ=false),RotationFilter=(bX=True,bY=True,bZ=True),ScaleFilter=(bX=false,bY=false,bZ=false))
    Parents=((Item=(Type=Bone,Name="ball_r"),Weight=1))
    Weight=1
   )
   > &RigUnit_ParentConstraint(
    AdvancedSettings=(InterpolationType=Average,RotationOrderForFilter=XZY)
    Child=(Type=Bone,Name="dyn_coatTweaker_cuff_thumb_l")
    Filter=(TranslationFilter=(bX=true,bY=true,bZ=true),RotationFilter=(bX=True,bY=True,bZ=True),ScaleFilter=(bX=false,bY=false,bZ=false))
    Parents=((Item=(Type=Bone,Name="thumb_02_l"),Weight=1))
    Weight=1
   )
   > &RigUnit_ParentConstraint(
    AdvancedSettings=(InterpolationType=Average,RotationOrderForFilter=XZY)
    Child=(Type=Bone,Name="dyn_coatTweaker_cuff_thumb_r")
    Filter=(TranslationFilter=(bX=true,bY=true,bZ=true),RotationFilter=(bX=True,bY=True,bZ=True),ScaleFilter=(bX=false,bY=false,bZ=false))
    Parents=((Item=(Type=Bone,Name="thumb_02_r"),Weight=1))
    Weight=1
   )
   > &RigUnit_ParentConstraint(
    AdvancedSettings=(InterpolationType=Average,RotationOrderForFilter=XZY)
    Child=(Type=Bone,Name="dyn_coatTweaker_cuff_hand_l")
    Filter=(TranslationFilter=(bX=true,bY=true,bZ=true),RotationFilter=(bX=True,bY=True,bZ=True),ScaleFilter=(bX=false,bY=false,bZ=false))
    Parents=((Item=(Type=Bone,Name="middle_metacarpal_l"),Weight=1))
    Weight=1
   )
   > &RigUnit_ParentConstraint(
    AdvancedSettings=(InterpolationType=Average,RotationOrderForFilter=XZY)
    Child=(Type=Bone,Name="dyn_coatTweaker_cuff_hand_r")
    Filter=(TranslationFilter=(bX=true,bY=true,bZ=true),RotationFilter=(bX=True,bY=True,bZ=True),ScaleFilter=(bX=false,bY=false,bZ=false))
    Parents=((Item=(Type=Bone,Name="middle_metacarpal_r"),Weight=1))
    Weight=1
   )
   n28 = RigVMVariableNode() {variable=CustomEmoteCrisp}
   Multiply = &RigVMFunction_MathDoubleMul(B=RigVMVariableNode() {variable=L calfBend Alpha}.Value, A=@n28.Value)
   n29 = &RigVMFunction_AlphaInterp(InterpSpeedDecreasing=30, InterpSpeedIncreasing=30, Scale=1, bClampResult=true, Value=@Multiply.Result)
   &RigUnit_ModifyTransforms(
    ItemToModify="((Item=(Type=Bone,Name=\"dyn_jacket_bck_c_base\"),Transform=(Rotation=(X=0,Y=-0.5373,Z=0,W=0.8434),Translation=(X=-3.7396,Y=0,Z=4.5139),Scale3D=(X=1,Y=1,Z=1))),(Item=(Type=Bone,Name=\"dyn_jacket_bck_l_base\"),Transform=(Rotation=(X=0,Y=-0.342,Z=0,W=0.9397),Tra ...(truncated)"
    WeightMinimum=0
    Weight=@n29.Result
    ExecutePin=@RigVMFunction_Sequence.F
   )
   n30 = &RigVMFunction_AlphaInterp(InterpSpeedDecreasing=30, InterpSpeedIncreasing=30, Scale=1, bClampResult=true, Value=@n28.Value)
   > &RigUnit_ModifyTransforms(
    ItemToModify="((Item=(Type=Bone,Name=\"dyn_jacket_bck_c_base\"),Transform=(Rotation=(X=0,Y=0,Z=0,W=1),Translation=(X=0,Y=0,Z=1.7223),Scale3D=(X=1,Y=1,Z=1))),(Item=(Type=Bone,Name=\"dyn_jacket_bck_l_base\"),Transform=(Rotation=(X=0,Y=0,Z=0,W=1),Translation=(X=0,Y=0,Z=4),Scal ...(truncated)"
    WeightMinimum=0
    Weight=@n30.Result
   )
   Multiply_2 = RigVMAggregateNode(
    B=RigVMVariableNode() {variable=L thighForward Alpha}.Value
    C=RigVMVariableNode() {variable=R thighForward Alpha}.Value
    A=RigVMVariableNode() {variable=CustomEmoteLime}.Value
    D=RigVMVariableNode() {variable=Low 2 Ground Lime Emote}.Value
   )
   n31 = &RigVMFunction_AlphaInterp(InterpSpeedDecreasing=30, InterpSpeedIncreasing=30, Scale=1, bClampResult=true, Value=@Multiply_2.Result)
   > &RigUnit_ModifyTransforms(
    ItemToModify="((Item=(Type=Bone,Name=\"dyn_jacket_bck_c_base\"),Transform=(Rotation=(X=0,Y=0.6756,Z=0,W=-0.7373),Translation=(X=-0.1186,Y=-0.0002,Z=15.0918),Scale3D=(X=1,Y=1,Z=1))),(Item=(Type=Bone,Name=\"dyn_jacket_bck_l_base\"),Transform=(Rotation=(X=0,Y=0.6756,Z=0,W=-0.7 ...(truncated)"
    WeightMinimum=0
    Weight=@n31.Result
   )
   n32 = &RigVMFunction_AlphaInterp(InterpSpeedDecreasing=30, InterpSpeedIncreasing=30, Scale=1, Value=0, bClampResult=true)
   > &RigUnit_ModifyTransforms(
    ItemToModify="((Item=(Type=Bone,Name=\"dyn_jacket_bck_c_base\"),Transform=(Rotation=(X=0,Y=0,Z=0,W=1),Translation=(X=0,Y=0,Z=0),Scale3D=(X=1,Y=1,Z=1))),(Item=(Type=Bone,Name=\"dyn_jacket_bck_l_base\"),Transform=(Rotation=(X=0,Y=0,Z=0,W=1),Translation=(X=0,Y=0,Z=0),Scale3D=( ...(truncated)"
    WeightMinimum=0
    Weight=@n32.Result
   )
   n33 = &RigVMFunction_AlphaInterp(
    InterpSpeedDecreasing=30
    InterpSpeedIncreasing=30
    Scale=1
    bClampResult=true
    Value=&RigVMFunction_MathDoubleMul(A=RigVMVariableNode() {variable=Emote Alpha}.Value, B=RigVMVariableNode() {variable=R thighSide Extreme Alpha}.Value).Result
   )
   > &RigUnit_ModifyTransforms(
    ItemToModify="((Item=(Type=Bone,Name=\"dyn_jacket_bck_c_base\"),Transform=(Rotation=(X=0.0207,Y=-0.2152,Z=-0.1284,W=0.9679),Translation=(X=-0.764,Y=0,Z=8.0986),Scale3D=(X=1,Y=1,Z=1))),(Item=(Type=Bone,Name=\"dyn_jacket_bck_l_base\"),Transform=(Rotation=(X=0,Y=0,Z=0,W=1),Tra ...(truncated)"
    WeightMinimum=0
    Weight=@n33.Result
   )
   n34 = &RigVMFunction_AlphaInterp(
    InterpSpeedDecreasing=30
    InterpSpeedIncreasing=30
    Scale=1
    bClampResult=true
    Value=&RigVMFunction_MathDoubleMul(B=RigVMVariableNode() {variable=L thighSide Extreme Alpha}.Value, A=RigVMVariableNode() {variable=Emote Alpha}.Value).Result
   )
   > &RigUnit_ModifyTransforms(
    ItemToModify="((Item=(Type=Bone,Name=\"dyn_jacket_bck_c_base\"),Transform=(Rotation=(X=0,Y=0,Z=0.0872,W=0.9962),Translation=(X=0,Y=0,Z=0),Scale3D=(X=1,Y=1,Z=1))),(Item=(Type=Bone,Name=\"dyn_jacket_bck_l_base\"),Transform=(Rotation=(X=-0.0038,Y=0.0871,Z=-0.0435,W=-0.9952),Tr ...(truncated)"
    WeightMinimum=0
    Weight=@n34.Result
   )
   &RigVMFunction_MathQuaternionInverse(Value=@n21.Twist) {template="Inverse::Execute(in Value,out Result)"}
   n35 = &RigUnit_GetRelativeTransformForItem(Child=(Type=Bone,Name="pelvis"), Parent=(Type=Bone,Name="root"), bChildInitial=False, bParentInitial=false)
   n36 = &RigVMFunction_MathDoubleRemap(
    SourceMaximum=60
    SourceMinimum=80
    TargetMaximum=1
    TargetMinimum=0
    Value=RigVMDispatchNode(Struct=@n35.RelativeTransform.Translation) {template="DISPATCH_RigVMDispatch_BreakStruct(in Struct,out Elements)"}.Elements.Z
   )
   n37 = &RigVMFunction_MathQuaternionToEuler(RotationOrder=ZYX, Value=@n35.RelativeTransform.Rotation)
   &RigVMFunction_MathDoubleMul(
    B=@n36.Result
    A=&RigVMFunction_MathDoubleRemap(
     SourceMaximum=69
     SourceMinimum=80
     TargetMaximum=1
     TargetMinimum=0
     Value=&RigVMFunction_MathDoubleAbs(Value=@n37.Result.Y).Result
    ).Result
   )
   n38 = &RigUnit_GetRelativeTransformForItem(Child=(Type=Bone,Name="pelvis"), Parent=(Type=Bone,Name="root"), bChildInitial=False, bParentInitial=false)
   n39 = &RigVMFunction_MathDoubleRemap(
    SourceMaximum=70
    SourceMinimum=100
    TargetMaximum=1
    TargetMinimum=0
    Value=RigVMDispatchNode(Struct=@n38.RelativeTransform.Translation) {template="DISPATCH_RigVMDispatch_BreakStruct(in Struct,out Elements)"}.Elements.Z
   )
   Absolute = &RigVMFunction_MathDoubleAbs(Value=&RigVMFunction_MathQuaternionToEuler(RotationOrder=ZYX, Value=@n38.RelativeTransform.Rotation).Result.Y)
   RigVMFunction_MathFloatMake = /Script/RigVM.RigVMFunction_MathFloatMake(Value=&RigVMFunction_MathDoubleSub(A=1, B=&RigVMFunction_MathBoolToFloat(Value=False).Result).Result) {is=struct RigVMUnitNode}
   &RigVMFunction_MathDoubleMul(
    A=&RigVMFunction_MathDoubleMul(
     B=@n39.Result
     A=&RigVMFunction_MathDoubleRemap(SourceMaximum=69, SourceMinimum=80, TargetMaximum=1, TargetMinimum=0, Value=@Absolute.Result).Result
    ).Result
    B=@RigVMFunction_MathFloatMake.Value
   )
   RigVMVariableNode() {variable=L calfBend Alpha}
   RigVMVariableNode() {variable=R calfBend Alpha}
   RigVMVariableNode(Value=@n39.Result, ExecuteContext=@n45.ExecuteContext) {variable=Low 2 Ground Lime Emote}
   RigVMFunction_MathBoolToFloat = &RigVMFunction_MathBoolToFloat(Value=False)
   Subtract = &RigVMFunction_MathDoubleSub(B=1, A=@RigVMFunction_MathBoolToFloat.Result)
   n40 = &RigVMFunction_MathDoubleAbs(Value=@Subtract.Result)
   n41 = &RigUnit_GetRelativeTransformForItem(Child=(Type=Bone,Name="calf_l"), Parent=(Type=Bone,Name="thigh_l"), bChildInitial=False, bParentInitial=False)
   n42 = &RigUnit_GetRelativeTransformForItem(Child=(Type=Bone,Name="calf_r"), Parent=(Type=Bone,Name="thigh_r"), bChildInitial=False, bParentInitial=False)
   n43 = &RigVMFunction_MathDoubleMul(B=@n40.Result, A=@n36.Result)
   RigVMFunction_MathQuaternionToEuler = &RigVMFunction_MathQuaternionToEuler(RotationOrder=ZYX, Value=@n41.RelativeTransform.Rotation)
   Remap = &RigVMFunction_MathDoubleRemap(SourceMaximum=100, SourceMinimum=45, TargetMaximum=1, TargetMinimum=0, Value=@RigVMFunction_MathQuaternionToEuler.Result.Z)
   n44 = &RigVMFunction_MathDoubleRemap(
    SourceMaximum=100
    SourceMinimum=45
    TargetMaximum=1
    TargetMinimum=0
    Value=&RigVMFunction_MathQuaternionToEuler(RotationOrder=ZYX, Value=@n42.RelativeTransform.Rotation).Result.Z
   )
   /Script/ControlRig.RigUnit_SetCurveValue(Curve=DeltaRef_Calf_L, Value=@Remap.Result, ExecutePin=@n46.ExecuteContext) {is=struct RigVMUnitNode}
   > RigVMVariableNode(Value=@n44.Result) {variable=R calfBend Alpha}
   > n45 = RigVMVariableNode(Value=@n43.Result) {variable=Low 2 Ground Alpha}
   n46 = RigVMVariableNode(Value=@Remap.Result, ExecuteContext=@n20.ExecutePin) {variable=L calfBend Alpha}
  ]}
]
hierarchy [
 root {loc=<0 0 0>, rot=<0 0 0 1>}
 pelvis {parent=root, loc=<0 -0.5301 98.0571>, rot=<0.0127 0.707 -0.0127 -0.707>}
 spine_01 {parent=pelvis, loc=<6.1208 -0.9328 0>, rot=<0 0 0.034 -0.9994>}
 spine_02 {parent=spine_01, loc=<6.5997 -0.2095 0>, rot=<0 0 0.0238 -0.9997>}
 spine_03 {parent=spine_02, loc=<6.4914 -0.1237 0>, rot=<0 0 0.0149 -0.9999>}
 spine_04 {parent=spine_03, loc=<6.9598 0.0794 0>, rot=<0 0 -0.0132 -0.9999>}
 spine_05 {parent=spine_04, loc=<6.9987 0 0>, rot=<0 0 -0.0015 -1>}
 neck_01 {parent=spine_05, loc=<11.3138 0.1529 0>, rot=<0 0 -0.2003 -0.9797>}
 neck_02 {parent=neck_01, loc=<4.5627 0 0>, rot=<0 0 0.0504 -0.9987>}
 head {parent=neck_02, loc=<5.3914 0 0>, rot=<0 0 0.115 -0.9934>}
 clavicle_l {parent=spine_05, loc=<5.4831 -0.1053 -3.2946>, rot=<0.0465 -0.7289 -0.009 -0.6829>}
 upperarm_l {parent=clavicle_l, loc=<9.8658 -1.2136 -0.4105>, rot=<0.0164 -0.3245 -0.0269 -0.9454>}
 lowerarm_l {parent=upperarm_l, loc=<26.465 -0.8345 0.259>, rot=<0.0036 0.0161 -0.2298 -0.9731>}
 hand_l {parent=lowerarm_l, loc=<22.7355 -0.3203 -0.6105>, rot=<-0.6784 -0.0223 0.0735 -0.7307>}
 thumb_01_l {parent=hand_l, loc=<1.3118 -0.3191 -2.459>, rot=<0.6546 -0.4042 0.0664 -0.6354>}
 thumb_02_l {parent=thumb_01_l, loc=<5.9336 -0.0448 0.0401>, rot=<-0.0037 0.0213 0.298 -0.9543>}
 thumb_03_l {parent=thumb_02_l, loc=<2.2223 0.3682 -0.0907>, rot=<0.0118 -0.013 0.0066 -0.9998>}
 index_metacarpal_l {parent=hand_l, loc=<1.6535 0.0183 -2.96>, rot=<0.0158 -0.0124 -0.0173 -0.9996>}
 index_01_l {parent=index_metacarpal_l, loc=<8.3008 0.0357 -0.0524>, rot=<-0.0044 0.0485 0.0924 -0.9945>}
 index_02_l {parent=index_01_l, loc=<3.3587 -0.3004 0.0452>, rot=<-0.002 0.0027 0.3225 -0.9466>}
 index_03_l {parent=index_02_l, loc=<2.5578 -0.123 -0.0252>, rot=<-0.002 0.0067 0.2558 -0.9667>}
 middle_metacarpal_l {parent=hand_l, loc=<2.1856 -0.1725 -0.5984>, rot=<0.006 0.0168 -0.0329 -0.9993>}
 middle_01_l {parent=middle_metacarpal_l, loc=<7.3275 -0.2447 0.0129>, rot=<-0.0208 0.0042 0.0858 -0.9961>}
 middle_02_l {parent=middle_01_l, loc=<4.0123 0.1401 0.1175>, rot=<-0.0218 0.0258 0.3512 -0.9357>}
 middle_03_l {parent=middle_02_l, loc=<2.6077 0.0534 0.1373>, rot=<0.0082 0.0188 0.1708 -0.9851>}
 ring_metacarpal_l {parent=hand_l, loc=<2.2074 -0.3625 1.3638>, rot=<-0.0284 0.0722 -0.0458 -0.9959>}
 ring_01_l {parent=ring_metacarpal_l, loc=<7.0295 -0.6197 -0.1446>, rot=<-0.0489 -0.0252 0.1892 -0.9804>}
 ring_02_l {parent=ring_01_l, loc=<3.6174 0.4647 -0.1027>, rot=<0 0 0.2806 -0.9598>}
 ring_03_l {parent=ring_02_l, loc=<2.4152 0.114 0.0199>, rot=<0 0 0.1189 -0.9929>}
 pinky_metacarpal_l {parent=hand_l, loc=<1.7616 -0.9753 3.3814>, rot=<-0.0397 0.0814 -0.069 -0.9935>}
 pinky_01_l {parent=pinky_metacarpal_l, loc=<7.1709 -0.4831 -0.1094>, rot=<-0.0654 -0.0035 0.2242 -0.9723>}
 pinky_02_l {parent=pinky_01_l, loc=<2.8274 0.0893 -0.1043>, rot=<0.0037 0.0011 0.2854 -0.9584>}
 pinky_03_l {parent=pinky_02_l, loc=<2.2046 0.0307 -0.0078>, rot=<-0.0002 0.0098 0.1445 -0.9895>}
 dyn_coatTweaker_cuff_thumb_l {parent=hand_l, loc=<3.941 -4.3638 -7.7811>, rot=<0.7578 -0.1927 0.213 -0.5859>}
 dyn_coatTweaker_cuff_hand_l {parent=hand_l, loc=<2.2518 2.9389 -0.3698>, rot=<0.0445 0.0344 -0.0404 -0.9976>}
 lowerarm_twist_02_l {parent=lowerarm_l, loc=<10.6707 0 0>, rot=<0 0 0 1>}
 lowerarm_twist_01_l {parent=lowerarm_l, loc=<18.316 0 0>, rot=<0 0 0 1>}
 dyn_coatTweaker_lowerarm_l {parent=lowerarm_l, loc=<12.3187 -2.7167 -6.308>, rot=<0.0367 -0.0081 -0.0245 -0.999>}
 upperarm_twist_01_l {parent=upperarm_l, loc=<7.2135 -0.1134 0>, rot=<0 0 0 1>}
 upperarm_twist_02_l {parent=upperarm_l, loc=<19.7998 -0.1134 0>, rot=<0 0 0 1>}
 deform_shoulder_l_base {parent=upperarm_l, loc=<6.1074 -0.1539 0.4519>, rot=<-0.0075 -0.0283 0.0011 -0.9996>}
 deform_shoulder_l {parent=deform_shoulder_l_base, loc=<0 0 0>, rot=<0 0 0 1>}
 dyn_coatTweaker_upperarm_l {parent=upperarm_l, loc=<21.2663 -0.3158 -6.4451>, rot=<-0.0074 0.0589 0.0017 -0.9982>}
 dyn_coat_lapel_l_top_base {parent=clavicle_l, loc=<12.5683 5.0546 0.1593>, rot=<-0.306 0.2164 -0.7117 -0.5942>}
 dyn_coat_lapel_l_top {parent=dyn_coat_lapel_l_top_base, loc=<0 0 0>, rot=<0 0 0 1>}
 clavicle_r {parent=spine_05, loc=<5.4831 -0.1053 3.2946>, rot=<0.7289 0.0465 -0.6829 0.009>}
 upperarm_r {parent=clavicle_r, loc=<-9.8658 1.2136 0.4105>, rot=<0.0164 -0.3245 -0.0269 -0.9454>}
 lowerarm_r {parent=upperarm_r, loc=<-26.465 0.8345 -0.259>, rot=<0.0036 0.0161 -0.2298 -0.9731>}
 hand_r {parent=lowerarm_r, loc=<-22.7355 0.3203 0.6106>, rot=<-0.6784 -0.0223 0.0735 -0.7307>}
 thumb_01_r {parent=hand_r, loc=<-1.3118 0.3191 2.459>, rot=<0.6546 -0.4042 0.0664 -0.6354>}
 thumb_02_r {parent=thumb_01_r, loc=<-5.9336 0.0448 -0.0401>, rot=<-0.0037 0.0213 0.298 -0.9543>}
 thumb_03_r {parent=thumb_02_r, loc=<-2.2223 -0.3682 0.0907>, rot=<0.0118 -0.013 0.0066 -0.9998>}
 index_metacarpal_r {parent=hand_r, loc=<-1.6535 -0.0183 2.96>, rot=<0.0158 -0.0124 -0.0173 -0.9996>}
 index_01_r {parent=index_metacarpal_r, loc=<-8.3008 -0.0357 0.0524>, rot=<-0.0044 0.0485 0.0924 -0.9945>}
 index_02_r {parent=index_01_r, loc=<-3.3587 0.3004 -0.0452>, rot=<-0.002 0.0027 0.3225 -0.9466>}
 index_03_r {parent=index_02_r, loc=<-2.5578 0.123 0.0252>, rot=<-0.002 0.0067 0.2558 -0.9667>}
 middle_metacarpal_r {parent=hand_r, loc=<-2.1856 0.1725 0.5984>, rot=<0.006 0.0168 -0.0329 -0.9993>}
 middle_01_r {parent=middle_metacarpal_r, loc=<-7.3275 0.2447 -0.0129>, rot=<-0.0208 0.0042 0.0858 -0.9961>}
 middle_02_r {parent=middle_01_r, loc=<-4.0123 -0.1401 -0.1175>, rot=<-0.0218 0.0258 0.3512 -0.9357>}
 middle_03_r {parent=middle_02_r, loc=<-2.6077 -0.0534 -0.1373>, rot=<0.0082 0.0188 0.1708 -0.9851>}
 ring_metacarpal_r {parent=hand_r, loc=<-2.2074 0.3625 -1.3638>, rot=<-0.0284 0.0722 -0.0458 -0.9959>}
 ring_01_r {parent=ring_metacarpal_r, loc=<-7.0295 0.6197 0.1446>, rot=<-0.0489 -0.0252 0.1892 -0.9804>}
 ring_02_r {parent=ring_01_r, loc=<-3.6174 -0.4647 0.1027>, rot=<0 0 0.2806 -0.9598>}
 ring_03_r {parent=ring_02_r, loc=<-2.4152 -0.114 -0.0199>, rot=<0 0 0.1189 -0.9929>}
 pinky_metacarpal_r {parent=hand_r, loc=<-1.7616 0.9753 -3.3814>, rot=<-0.0397 0.0814 -0.069 -0.9935>}
 pinky_01_r {parent=pinky_metacarpal_r, loc=<-7.1709 0.4831 0.1094>, rot=<-0.0654 -0.0035 0.2242 -0.9723>}
 pinky_02_r {parent=pinky_01_r, loc=<-2.8274 -0.0893 0.1043>, rot=<0.0037 0.0011 0.2854 -0.9584>}
 pinky_03_r {parent=pinky_02_r, loc=<-2.2046 -0.0307 0.0078>, rot=<-0.0002 0.0098 0.1445 -0.9895>}
 dyn_coatTweaker_cuff_thumb_r {parent=hand_r, loc=<-3.9408 4.3641 7.7813>, rot=<0.7428 -0.2024 0.2427 -0.5902>}
 dyn_coatTweaker_cuff_hand_r {parent=hand_r, loc=<-2.7996 -3.0843 0.2542>, rot=<0.0187 0.0221 -0.0343 -0.999>}
 lowerarm_twist_02_r {parent=lowerarm_r, loc=<-10.6707 0 0>, rot=<0 0 0 -1>}
 lowerarm_twist_01_r {parent=lowerarm_r, loc=<-18.316 0 0>, rot=<0 0 0 -1>}
 dyn_coatTweaker_lowerarm_r {parent=lowerarm_r, loc=<-10.9055 0.5435 5.8466>, rot=<0.0073 -0.0016 -0.0112 -0.9999>}
 upperarm_twist_01_r {parent=upperarm_r, loc=<-7.2135 0.1134 0>, rot=<0 0 0 -1>}
 upperarm_twist_02_r {parent=upperarm_r, loc=<-19.7998 0.1134 0>, rot=<0 0 0 -1>}
 deform_shoulder_r_base {parent=upperarm_r, loc=<-6.107 0.1539 -0.4516>, rot=<-0.0011 -0.9996 -0.0075 0.0283>}
 deform_shoulder_r {parent=deform_shoulder_r_base, loc=<0 0 0>, rot=<0 0 0 1>}
 dyn_coatTweaker_upperarm_r {parent=upperarm_r, loc=<-21.2666 0.3158 6.4448>, rot=<-0.0075 0.0153 0.0014 -0.9999>}
 dyn_coat_lapel_r_top_base {parent=clavicle_r, loc=<-12.568 -5.0545 -0.1593>, rot=<0.6899 -0.6156 -0.3038 -0.2299>}
 dyn_coat_lapel_r_top {parent=dyn_coat_lapel_r_top_base, loc=<0 0 0>, rot=<0 0 0 1>}
 attach_backpack {parent=spine_05, loc=<-0.1233 -13.4384 -0.0013>, rot=<-0.0093 -0.707 0.0093 -0.707>}
 Backpack_BR {parent=attach_backpack, type=Null, loc=<0 2.4205 -0.1479>, rot=<-0.0872 0 0 0.9962>}
 attach_cape {parent=spine_05, loc=<9.0066 -11.6831 0.0105>, rot=<0.0616 -0.7044 -0.0616 -0.7044>}
 cape {parent=attach_cape, type=Null, loc=<0 0 0>, rot=<0 0 0 1>}
 dyn_necklaces_base {parent=spine_05, loc=<8.1084 2.2218 0>, rot=<0.2442 0.6636 0.6636 0.2442>}
 dyn_necklace_A {parent=dyn_necklaces_base, loc=<0 0 0>, rot=<0 0 0 1>}
 dyn_necklace_B_01 {parent=dyn_necklaces_base, loc=<1.125 0 0.0523>, rot=<0.0001 0.0679 0 -0.9977>}
 dyn_necklace_B_02 {parent=dyn_necklace_B_01, loc=<6.0148 0 0>, rot=<0 -0.0436 0 -0.999>}
 dyn_necklace_C_01 {parent=dyn_necklaces_base, loc=<2.2851 0 0.0389>, rot=<-0.0001 0.0645 -0.0001 -0.9979>}
 dyn_necklace_C_02 {parent=dyn_necklace_C_01, loc=<7.1175 0 0>, rot=<0 -0.0872 0 -0.9962>}
 deform_coat_lapel_l_mid_base {parent=spine_04, loc=<2.0677 8.1387 -15.194>, rot=<0.6008 0.4882 0.4658 0.4286>}
 deform_coat_lapel_l_mid {parent=deform_coat_lapel_l_mid_base, loc=<0 0 0>, rot=<0 0 0 1>}
 deform_coat_lapel_r_mid_base {parent=spine_04, loc=<2.0677 8.1387 15.1936>, rot=<-0.4286 -0.4658 -0.4882 -0.6008>}
 deform_coat_lapel_r_mid {parent=deform_coat_lapel_r_mid_base, loc=<0 0 0>, rot=<0 0 0 1>}
 deform_armpit_l_base {parent=spine_03, loc=<-0.141 -0.3667 -14.814>, rot=<0.0387 -0.706 -0.0387 -0.706>}
 deform_armpit_l {parent=deform_armpit_l_base, loc=<0 0 0>, rot=<0 0 0 1>}
 deform_armpit_r_base {parent=spine_03, loc=<-0.141 -0.3667 14.8137>, rot=<0.706 0.0387 0.706 -0.0387>}
 deform_armpit_r {parent=deform_armpit_r_base, loc=<0 0 0>, rot=<0 0 0 1>}
 deform_coat_lapel_l_bot_base {parent=spine_03, loc=<-2.0444 10.4453 -12.925>, rot=<0.5318 0.5092 0.5462 0.3995>}
 deform_coat_lapel_l_bot {parent=deform_coat_lapel_l_bot_base, loc=<-0.2983 -1.6821 0.1443>, rot=<0 0 0 1>}
 deform_coat_lapel_r_bot_base {parent=spine_03, loc=<-2.0444 10.4453 12.9248>, rot=<-0.3995 -0.5462 -0.5092 -0.5318>}
 deform_coat_lapel_r_bot {parent=deform_coat_lapel_r_bot_base, loc=<0 0 0>, rot=<0 0 0 1>}
 dyn_jacket_fr_l_base {parent=spine_01, loc=<-0.0666 12.9429 -13.4684>, rot=<0.0161 -0.8865 -0.4615 -0.031>}
 dyn_jacket_fr_l_01 {parent=dyn_jacket_fr_l_base, loc=<0 0 0>, rot=<0.0011 -0.0019 0.0006 -1>}
 dyn_jacket_fr_l_02 {parent=dyn_jacket_fr_l_01, loc=<16.3079 0.5928 -0.4486>, rot=<0.0222 -0.0194 -0.0085 -0.9995>}
 dyn_jacket_fr_l_03 {parent=dyn_jacket_fr_l_02, loc=<14.9192 0.7793 -0.2273>, rot=<-0.0016 0.0106 0 -0.9999>}
 dyn_jacket_fr_l_04 {parent=dyn_jacket_fr_l_03, loc=<14.419 0.4498 -0.5549>, rot=<0.0126 -0.004 0.0157 -0.9998>}
 dyn_jacket_fr_l_05 {parent=dyn_jacket_fr_l_04, loc=<14.408 0 0>, rot=<0 0 0 1>}
 dyn_jacket_side_l_base {parent=spine_01, loc=<-0.5225 0.8579 -17.7614>, rot=<-0.0104 -0.9969 0.001 -0.0783>}
 dyn_jacket_side_l_01 {parent=dyn_jacket_side_l_base, loc=<0 0 0>, rot=<0.0043 -0.0009 0.0018 -1>}
 dyn_jacket_side_l_02 {parent=dyn_jacket_side_l_01, loc=<15.4907 -0.039 -1.4782>, rot=<-0.0143 -0.0255 0.0247 -0.9993>}
 dyn_jacket_side_l_03 {parent=dyn_jacket_side_l_02, loc=<13.9793 -0.0059 -0.2152>, rot=<0.0308 0.0103 0.0166 -0.9993>}
 dyn_jacket_side_l_04 {parent=dyn_jacket_side_l_03, loc=<14.4878 0.0096 -0.3128>, rot=<0.014 0.0137 -0.0014 -0.9998>}
 dyn_jacket_side_l_05 {parent=dyn_jacket_side_l_04, loc=<14.525 0 0>, rot=<0 0 0 1>}
 dyn_jacket_bck_l_base {parent=spine_01, loc=<-0.5806 -10.6631 -11.8773>, rot=<0.0387 -0.8837 0.4619 -0.0644>}
 dyn_jacket_bck_l_01 {parent=dyn_jacket_bck_l_base, loc=<0 0 0>, rot=<0.0005 0.0003 0.0012 -1>}
 dyn_jacket_bck_l_02 {parent=dyn_jacket_bck_l_01, loc=<15.7415 0.0229 -0.3382>, rot=<0.0285 -0.0121 -0.0241 -0.9992>}
 dyn_jacket_bck_l_03 {parent=dyn_jacket_bck_l_02, loc=<13.9817 0.1215 -0.0421>, rot=<-0.007 0.0155 -0.0107 -0.9998>}
 dyn_jacket_bck_l_04 {parent=dyn_jacket_bck_l_03, loc=<13.6917 -0.1812 -0.6856>, rot=<-0.0172 0.0234 0.0025 -0.9996>}
 dyn_jacket_bck_l_05 {parent=dyn_jacket_bck_l_04, loc=<14.9138 -0.5826 -0.9702>, rot=<0 0 0 1>}
 dyn_jacket_bck_c_base {parent=spine_01, loc=<-0.2179 -13.4681 -0.0291>, rot=<0.0606 -0.7033 0.7057 -0.0608>}
 dyn_jacket_bck_c_01 {parent=dyn_jacket_bck_c_base, loc=<0 0 0>, rot=<0.0017 0.0022 0 -1>}
 dyn_jacket_bck_c_02 {parent=dyn_jacket_bck_c_01, loc=<16.7909 0 -0.1463>, rot=<0 -0.0153 0 -0.9999>}
 dyn_jacket_bck_c_03 {parent=dyn_jacket_bck_c_02, loc=<13.7725 0 0>, rot=<0 -0.0172 0 -0.9999>}
 dyn_jacket_bck_c_04 {parent=dyn_jacket_bck_c_03, loc=<13.4707 0 0>, rot=<-0.0061 -0.0144 -0.0059 -0.9999>}
 dyn_jacket_bck_c_05 {parent=dyn_jacket_bck_c_04, loc=<14.9719 0.0082 0.7275>, rot=<0.0087 0 0 -1>}
 dyn_jacket_bck_r_base {parent=spine_01, loc=<-0.5806 -10.6631 11.877>, rot=<0.0607 -0.3829 0.9207 -0.0442>}
 dyn_jacket_bck_r_01 {parent=dyn_jacket_bck_r_base, loc=<0 0 0>, rot=<-0.0003 0.0001 -0.0012 -1>}
 dyn_jacket_bck_r_02 {parent=dyn_jacket_bck_r_01, loc=<15.7806 -0.2893 0.011>, rot=<0.0153 -0.0088 0.0254 -0.9995>}
 dyn_jacket_bck_r_03 {parent=dyn_jacket_bck_r_02, loc=<14.0137 0.2957 0.4275>, rot=<0.0067 0.0164 0.0093 -0.9998>}
 dyn_jacket_bck_r_04 {parent=dyn_jacket_bck_r_03, loc=<13.7362 0.4237 -0.2464>, rot=<0.0182 0.023 -0.0045 -0.9996>}
 dyn_jacket_bck_r_05 {parent=dyn_jacket_bck_r_04, loc=<14.6614 0.5708 -0.7005>, rot=<0 0 0 1>}
 dyn_jacket_side_r_base {parent=spine_01, loc=<-0.5225 0.8579 16.8852>, rot=<0.0789 0.047 0.9957 0.0069>}
 dyn_jacket_side_r_01 {parent=dyn_jacket_side_r_base, loc=<0 0 0>, rot=<0.0003 -0.0012 -0.002 -1>}
 dyn_jacket_side_r_02 {parent=dyn_jacket_side_r_01, loc=<15.5818 -0.0347 -0.8951>, rot=<-0.0295 -0.0285 -0.0212 -0.9989>}
 dyn_jacket_side_r_03 {parent=dyn_jacket_side_r_02, loc=<13.9763 -0.0359 -0.2419>, rot=<0.0128 0.008 -0.0178 -0.9997>}
 dyn_jacket_side_r_04 {parent=dyn_jacket_side_r_03, loc=<14.5269 0 0>, rot=<-0.0141 0.0137 0.0002 -0.9998>}
 dyn_jacket_side_r_05 {parent=dyn_jacket_side_r_04, loc=<14.5249 0 0>, rot=<-0.0436 0 0 -0.999>}
 dyn_jacket_fr_r_base {parent=spine_01, loc=<-0.0666 12.9429 13>, rot=<0.0296 0.4233 0.9054 -0.0171>}
 dyn_jacket_fr_r_01 {parent=dyn_jacket_fr_r_base, loc=<0 0 0>, rot=<0 0.001 0.0005 -1>}
 dyn_jacket_fr_r_02 {parent=dyn_jacket_fr_r_01, loc=<16.2917 -0.7327 -0.6501>, rot=<0.0214 -0.0204 0.0059 -0.9995>}
 dyn_jacket_fr_r_03 {parent=dyn_jacket_fr_r_02, loc=<14.9463 -0.2414 0.2939>, rot=<0.0017 0.0105 0.0018 -0.9999>}
 dyn_jacket_fr_r_04 {parent=dyn_jacket_fr_r_03, loc=<14.4562 0.2951 -0.0645>, rot=<0.0307 -0.0005 -0.0162 -0.9994>}
 dyn_jacket_fr_r_05 {parent=dyn_jacket_fr_r_04, loc=<14.4043 -0.1372 0.0552>, rot=<0 0 0 1>}
 thigh_l {parent=pelvis, loc=<-1.0855 -0.2109 -8.7053>, rot=<0.0009 0.0355 0.0155 -0.9993>}
 calf_l {parent=thigh_l, loc=<-42.5284 -0.0019 -0.7297>, rot=<0.0006 -0.0161 -0.0391 -0.9991>}
 foot_l {parent=calf_l, loc=<-46.7401 -2.5014 -0.592>, rot=<0.0427 -0.0238 0.0353 -0.9982>}
 ball_l {parent=foot_l, loc=<-3.5658 10.8066 -0.1544>, rot=<0.0001 -0.0001 -0.7063 -0.7079>}
 deform_ball_l {parent=foot_l, loc=<-5.6289 5.7699 0.1878>, rot=<0.0312 -0.0281 -0.7027 -0.7103>}
 calf_twist_02_l {parent=calf_l, loc=<-19.1546 -1.7943 -0.2638>, rot=<0.0237 0.0095 -0.0035 -0.9997>}
 calf_twist_01_l {parent=calf_l, loc=<-35.8483 -1.9197 -0.5794>, rot=<0.0237 0.0095 -0.0035 -0.9997>}
 thigh_twist_01_l {parent=thigh_l, loc=<-27.952 0 0>, rot=<0.0013 0 0 -1>}
 thigh_r {parent=pelvis, loc=<-1.0855 -0.2109 8.7053>, rot=<0.0355 -0.0009 0.9993 0.0155>}
 calf_r {parent=thigh_r, loc=<42.5284 0.0019 0.7297>, rot=<0.0006 -0.0161 -0.0391 -0.9991>}
 foot_r {parent=calf_r, loc=<46.7401 2.5014 0.592>, rot=<0.0427 -0.0238 0.0353 -0.9982>}
 ball_r {parent=foot_r, loc=<3.5658 -10.8066 0.1544>, rot=<0.0001 -0.0001 -0.7063 -0.7079>}
 deform_ball_r {parent=foot_r, loc=<7.0673 -6.5473 -0.2504>, rot=<0.0005 0.0029 -0.7034 -0.7108>}
 calf_twist_02_r {parent=calf_r, loc=<19.1546 1.7943 0.2638>, rot=<0.0237 0.0095 -0.0035 -0.9997>}
 calf_twist_01_r {parent=calf_r, loc=<35.8483 1.9197 0.5794>, rot=<0.0237 0.0095 -0.0035 -0.9997>}
 thigh_twist_01_r {parent=thigh_r, loc=<27.952 0 0>, rot=<0.0013 0 0 -1>}
 DeltaRef_Calf_L {type=Curve, loc=<0 0 0>, rot=<0 0 0 1>}
]
variables [
 CustomEmoteCrisp {type=double, default=0}
 CustomEmoteLime {type=double, default=0}
 Emote Alpha {type=double, default=0}
 FrontEnd Alpha {type=double, default=0}
 L calfBend Alpha {type=float, default=0}
 L thighBack Alpha {type=double, default=0}
 L thighForward Alpha {type=double, default=0}
 L thighSide Alpha {type=double, default=0}
 L thighSide Extreme Alpha {type=double, default=0}
 L thighSideIn Alpha {type=double, default=0}
 L thighTwistIn Alpha {type=double, default=0}
 Low 2 Ground Alpha {type=double, default=0}
 Low 2 Ground Lime Emote {type=double, default=0}
 R calfBend Alpha {type=float, default=0}
 R thighBack Alpha {type=double, default=0}
 R thighForward Alpha {type=double, default=0}
 R thighSide Alpha {type=double, default=0}
 R thighSide Extreme Alpha {type=double, default=0}
 R thighSideIn Alpha {type=double, default=0}
 R thighTwistIn Alpha {type=double, default=0}
]
