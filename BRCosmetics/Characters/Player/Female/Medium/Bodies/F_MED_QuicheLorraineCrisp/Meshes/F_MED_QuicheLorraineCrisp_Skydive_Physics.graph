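# Skydive physics asset for the QuicheLorraineCrisp female character body, simulating the jacket's hanging panels against the lower body.
# Capsules on pelvis, spine_01, thighs, calves and feet accompany seven jacket bone chains, with extra boxes on many lower segments.
# Chain bases join spine_01 with free rotation, while links below allow wide swing on one axis and little twist.
# Free-rotation cross constraints tie lower front segments to matching side segments.
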PhysicsAsset /BRCosmetics/Characters/Player/Female/Medium/Bodies/F_MED_QuicheLorraineCrisp/Meshes/F_MED_QuicheLorraineCrisp_Skydive_Physics
bodies {
 calf_l [{shapeType=Capsule, length=49, radius=12, shapeName=calf_l_capsule, center=<-22.3633 0.0001 0>, rotation=<rot -90 -0.0003 0>}]
 calf_r [{shapeType=Capsule, length=49, radius=12, shapeName=calf_r_capsule, center=<22.3633 -0.0001 0>, rotation=<rot 90 -0.0001 0>}]
 dyn_jacket_bck_c_01 [{shapeType=Capsule, length=13, radius=15, shapeName=dyn_jacket_bck_c_01_capsule, center=<6 0 18.6099>, rotation=<rot 90 0 0>}]
 dyn_jacket_bck_c_02 [{shapeType=Capsule, length=12, radius=16, shapeName=dyn_jacket_bck_c_02_capsule, center=<6 0 20>, rotation=<rot 90 0 0>}]
 dyn_jacket_bck_c_03 [{shapeType=Capsule, length=8, radius=20, shapeName=dyn_jacket_bck_c_03_capsule, center=<6 0 20>, rotation=<rot 90 0 0>}]
 dyn_jacket_bck_c_04 [
  {shapeType=Capsule, length=8, radius=20, shapeName=dyn_jacket_bck_c_04_capsule, center=<6 0 20>, rotation=<rot 90 0 0>}
  {shapeType=Box, extentX=44, extentY=81, extentZ=33, shapeName=dyn_jacket_bck_c_04_box, center=<10.8665 0 13>}
 ]
 dyn_jacket_bck_c_05 [
  {shapeType=Capsule, length=8, radius=20, shapeName=dyn_jacket_bck_c_05_capsule, center=<6 0 20>, rotation=<rot 90 0 0>}
  {shapeType=Box, extentX=57, extentY=131, extentZ=48, shapeName=dyn_jacket_bck_c_05_box, center=<10.8665 0 13>}
 ]
 dyn_jacket_bck_c_base [{shapeType=Capsule, length=8, radius=12, shapeName=dyn_jacket_bck_c_base_capsule, center=<6 0 13>}]
 dyn_jacket_bck_l_01 [{shapeType=Capsule, length=15.5, radius=12.5, shapeName=dyn_jacket_bck_l_01_capsule, center=<6 0 18.6099>, rotation=<rot 90 0.0003 0>}]
 dyn_jacket_bck_l_02 [{shapeType=Capsule, length=12, radius=16, shapeName=dyn_jacket_bck_l_02_capsule, center=<6 0 20>, rotation=<rot 90 0.0004 0>}]
 dyn_jacket_bck_l_03 [{shapeType=Capsule, length=8, radius=20, shapeName=dyn_jacket_bck_l_03_capsule, center=<6 0 20>, rotation=<rot 90 0.0004 0>}]
 dyn_jacket_bck_l_04 [
  {shapeType=Capsule, length=8, radius=20, shapeName=dyn_jacket_bck_l_04_capsule, center=<6 0 20>, rotation=<rot 90 0.0004 0>}
  {shapeType=Box, extentX=36, extentY=73, extentZ=24, shapeName=dyn_jacket_bck_l_04_box, center=<10.8665 0 13>}
 ]
 dyn_jacket_bck_l_05 [
  {shapeType=Capsule, length=8, radius=20, shapeName=dyn_jacket_bck_l_05_capsule, center=<6 0 20>, rotation=<rot 90 0.0004 0>}
  {shapeType=Box, extentX=41, extentY=81, extentZ=28, shapeName=dyn_jacket_bck_l_05_box, center=<10.8665 0 13>}
 ]
 dyn_jacket_bck_l_base [{shapeType=Capsule, length=8, radius=12, shapeName=dyn_jacket_bck_l_base_capsule, center=<6 0 13>}]
 dyn_jacket_bck_r_01 [{shapeType=Capsule, length=15.5, radius=12.5, shapeName=dyn_jacket_bck_r_01_capsule, center=<6 0 18.6099>, rotation=<rot 90 -0.0004 0>}]
 dyn_jacket_bck_r_02 [{shapeType=Capsule, length=12, radius=16, shapeName=dyn_jacket_bck_r_02_capsule, center=<6 0 20>, rotation=<rot 90 -0.0004 0>}]
 dyn_jacket_bck_r_03 [{shapeType=Capsule, length=8, radius=20, shapeName=dyn_jacket_bck_r_03_capsule, center=<6 0 20>, rotation=<rot 90 -0.0004 0>}]
 dyn_jacket_bck_r_04 [
  {shapeType=Capsule, length=8, radius=20, shapeName=dyn_jacket_bck_r_04_capsule, center=<6 0 20>, rotation=<rot 90 -0.0004 0>}
  {shapeType=Box, extentX=36, extentY=73, extentZ=25, shapeName=dyn_jacket_bck_r_04_box, center=<10.8665 0 13>}
 ]
 dyn_jacket_bck_r_05 [
  {shapeType=Capsule, length=8, radius=20, shapeName=dyn_jacket_bck_r_05_capsule, center=<6 0 20>, rotation=<rot 90 -0.0004 0>}
  {shapeType=Box, extentX=41, extentY=81, extentZ=29, shapeName=dyn_jacket_bck_r_05_box, center=<10.8665 0 13>}
 ]
 dyn_jacket_bck_r_base [{shapeType=Capsule, length=8, radius=12, shapeName=dyn_jacket_bck_r_base_capsule, center=<6 0 13>}]
 dyn_jacket_fr_l_01 [{shapeType=Capsule, length=12, radius=14, shapeName=dyn_jacket_fr_l_01_capsule, center=<6 0 20>, rotation=<rot 90 0.0003 0>}]
 dyn_jacket_fr_l_02 [{shapeType=Capsule, length=8, radius=18, shapeName=dyn_jacket_fr_l_02_capsule, center=<6 0 20>, rotation=<rot 90 0.0003 0>}]
 dyn_jacket_fr_l_03 [{shapeType=Capsule, length=8, radius=18, shapeName=dyn_jacket_fr_l_03_capsule, center=<6 0 20>, rotation=<rot 90 0.0003 0>}]
 dyn_jacket_fr_l_04 [{shapeType=Capsule, length=11, radius=19, shapeName=dyn_jacket_fr_l_04_capsule, center=<12.1047 0 20>, rotation=<rot 90 0.0003 0>}]
 dyn_jacket_fr_l_05 [{shapeType=Capsule, length=11, radius=19, shapeName=dyn_jacket_fr_l_05_capsule, center=<12.1047 0 20>, rotation=<rot 90 0.0003 0>}]
 dyn_jacket_fr_l_base [{shapeType=Capsule, length=8, radius=12, shapeName=dyn_jacket_fr_l_base_capsule, center=<6 0 13>}]
 dyn_jacket_fr_r_01 [{shapeType=Capsule, length=12, radius=14, shapeName=dyn_jacket_fr_r_01_capsule, center=<6 0 20>, rotation=<rot 90 -0.0004 0>}]
 dyn_jacket_fr_r_02 [
  {shapeType=Capsule, length=8, radius=18, shapeName=dyn_jacket_fr_r_02_capsule, center=<6 0 23>, rotation=<rot 90 -0.0004 0>}
  {
   shapeType=Box
   extentX=36
   extentY=72
   extentZ=43
   shapeName=dyn_jacket_fr_r_02_box
   center=<11.1975 1.5885 22.9774>
   rotation=<rot 0.0001 -0.0001 20>
  }
 ]
 dyn_jacket_fr_r_03 [
  {shapeType=Capsule, length=8, radius=18, shapeName=dyn_jacket_fr_r_03_capsule, center=<6 0 23>, rotation=<rot 90 -0.0004 0>}
  {shapeType=Box, extentX=36, extentY=78, extentZ=49, shapeName=dyn_jacket_fr_r_03_box, center=<11.1975 1.5967 23>, rotation=<rot 0.0001 0 20>}
 ]
 dyn_jacket_fr_r_04 [
  {shapeType=Capsule, length=15, radius=18, shapeName=dyn_jacket_fr_r_04_capsule, center=<6 0 23>, rotation=<rot 90 -0.0004 0>}
  {shapeType=Box, extentX=36, extentY=72, extentZ=34, shapeName=dyn_jacket_fr_r_04_box, center=<11.1975 1.5967 23>, rotation=<rot 0 -0.0001 20>}
 ]
 dyn_jacket_fr_r_05 [
  {shapeType=Capsule, length=8, radius=18, shapeName=dyn_jacket_fr_r_05_capsule, center=<6 0 23>, rotation=<rot 90 -0.0004 0>}
  {shapeType=Box, extentX=46, extentY=77, extentZ=51, shapeName=dyn_jacket_fr_r_05_box, center=<11.1975 1.5967 23>, rotation=<rot 0 -0.0001 20>}
 ]
 dyn_jacket_fr_r_base [{shapeType=Capsule, length=8, radius=12, shapeName=dyn_jacket_fr_r_base_capsule, center=<6 0 13>}]
 dyn_jacket_side_l_01 [{shapeType=Capsule, length=10.5, radius=12.5, shapeName=dyn_jacket_side_l_01_capsule, center=<6 0 15>, rotation=<rot 90 0.0006 0>}]
 dyn_jacket_side_l_02 [
  {shapeType=Capsule, length=14, radius=15, shapeName=dyn_jacket_side_l_02_capsule, center=<6 0 20>, rotation=<rot 90 0.0006 0>}
  {shapeType=Box, extentX=42, extentY=73, extentZ=19, shapeName=dyn_jacket_side_l_02_box, center=<14.0321 0 17.1626>}
 ]
 dyn_jacket_side_l_03 [
  {shapeType=Capsule, length=8, radius=15, shapeName=dyn_jacket_side_l_03_capsule, center=<6 0 20>, rotation=<rot 90 0.0005 0>}
  {shapeType=Box, extentX=42, extentY=73, extentZ=20, shapeName=dyn_jacket_side_l_03_box, center=<15.1042 0 21.1626>}
 ]
 dyn_jacket_side_l_04 [
  {shapeType=Capsule, length=8, radius=15, shapeName=dyn_jacket_side_l_04_capsule, center=<6 0 20>, rotation=<rot 90 0.0006 0>}
  {
   shapeType=Box
   extentX=41.5
   extentY=77.5
   extentZ=23.5
   shapeName=dyn_jacket_side_l_04_box
   center=<10.8665 -0.1907 19.0815>
   rotation=<rot 0 0.0002 -10>
  }
 ]
 dyn_jacket_side_l_05 [
  {shapeType=Capsule, length=8, radius=15, shapeName=dyn_jacket_side_l_05_capsule, center=<6 0 20>, rotation=<rot 90 0.0006 0>}
  {
   shapeType=Box
   extentX=41.5
   extentY=77.5
   extentZ=28.5
   shapeName=dyn_jacket_side_l_05_box
   center=<10.8665 0.0688 17.6098>
   rotation=<rot 0 0.0002 -10>
  }
 ]
 dyn_jacket_side_l_base [{shapeType=Capsule, length=8, radius=12, shapeName=dyn_jacket_side_l_base_capsule, center=<6 0 13>}]
 dyn_jacket_side_r_01 [{shapeType=Capsule, length=9.5, radius=13.5, shapeName=dyn_jacket_side_r_01_capsule, center=<6 0 16.6599>, rotation=<rot 90 -0.0006 0>}]
 dyn_jacket_side_r_02 [
  {shapeType=Capsule, length=14, radius=15, shapeName=dyn_jacket_side_r_02_capsule, center=<6 0 20>, rotation=<rot 90 -0.0006 0>}
  {shapeType=Box, extentX=42, extentY=73, extentZ=16, shapeName=dyn_jacket_side_r_02_box, center=<13.2555 0 14>}
 ]
 dyn_jacket_side_r_03 [
  {shapeType=Capsule, length=8, radius=15, shapeName=dyn_jacket_side_r_03_capsule, center=<6 0 20>, rotation=<rot 90 -0.0006 0>}
  {shapeType=Box, extentX=42, extentY=73, extentZ=21, shapeName=dyn_jacket_side_r_03_box, center=<13.7715 0 14>}
 ]
 dyn_jacket_side_r_04 [
  {shapeType=Capsule, length=8, radius=15, shapeName=dyn_jacket_side_r_04_capsule, center=<6 0 20>, rotation=<rot 90 -0.0006 0>}
  {shapeType=Box, extentX=42, extentY=73, extentZ=18, shapeName=dyn_jacket_side_r_04_box, center=<14.5628 0 13>, rotation=<rot 0 -0.0002 10>}
 ]
 dyn_jacket_side_r_05 [
  {shapeType=Capsule, length=8, radius=15, shapeName=dyn_jacket_side_r_05_capsule, center=<6 0 20>, rotation=<rot 90 -0.0006 0>}
  {shapeType=Box, extentX=42, extentY=73, extentZ=18, shapeName=dyn_jacket_side_r_05_box, center=<10.8665 0 13>, rotation=<rot 0 -0.0002 10>}
 ]
 dyn_jacket_side_r_base [{shapeType=Capsule, length=8, radius=12, shapeName=dyn_jacket_side_r_base_capsule, center=<6 0 13>}]
 foot_l [{shapeType=Capsule, length=1.5, radius=21.5, shapeName=foot_l_capsule, center=<-2.2314 1.3867 0>, rotation=<rot 0 -0.0001 -90>}]
 foot_r [{shapeType=Capsule, length=1.5, radius=21.5, shapeName=foot_r_capsule, center=<2.2314 -2.2071 0>, rotation=<rot 0 179.9998 -90>}]
 pelvis [{shapeType=Capsule, length=6.5, radius=12.2444, shapeName=pelvis_capsule, center=<-2.4456 0 0>}]
 spine_01 [{shapeType=Capsule, length=0.505, radius=0.505, shapeName=spine_01_capsule}]
 thigh_l [{shapeType=Capsule, length=28.5, radius=13, shapeName=thigh_l_capsule, center=<-28.3449 0.0003 0>, rotation=<rot 90 -0.0006 0>}]
 thigh_r [{shapeType=Capsule, length=28.5, radius=13, shapeName=thigh_r_capsule, center=<28.3449 -0.0001 0>, rotation=<rot -90 -0.0002 0>}]
}
constraints [
 {bone1Name=dyn_jacket_side_r_05, bone2Name=dyn_jacket_side_r_04, swing1LimitDegrees=15, swing2LimitDegrees=100, twistLimitDegrees=5}
 {
  bone1Name=dyn_jacket_bck_c_base
  bone2Name=spine_01
  swing1LimitDegrees=45
  swing1Motion=Free
  swing2LimitDegrees=45
  swing2Motion=Free
  twistLimitDegrees=45
  twistMotion=Free
 }
 {bone1Name=dyn_jacket_bck_c_01, bone2Name=dyn_jacket_bck_c_base, swing1LimitDegrees=10, swing2LimitDegrees=80, twistLimitDegrees=1}
 {bone1Name=dyn_jacket_bck_c_02, bone2Name=dyn_jacket_bck_c_01, swing1LimitDegrees=10, swing2LimitDegrees=80, twistLimitDegrees=1}
 {bone1Name=dyn_jacket_bck_c_03, bone2Name=dyn_jacket_bck_c_02, swing1LimitDegrees=10, swing2LimitDegrees=100, twistLimitDegrees=1}
 {bone1Name=dyn_jacket_bck_c_04, bone2Name=dyn_jacket_bck_c_03, swing1LimitDegrees=10, swing2LimitDegrees=100, twistLimitDegrees=1}
 {bone1Name=dyn_jacket_bck_c_05, bone2Name=dyn_jacket_bck_c_04, swing1LimitDegrees=10, swing2LimitDegrees=100, twistLimitDegrees=1}
 {
  bone1Name=dyn_jacket_bck_l_base
  bone2Name=spine_01
  swing1LimitDegrees=45
  swing1Motion=Free
  swing2LimitDegrees=45
  swing2Motion=Free
  twistLimitDegrees=45
  twistMotion=Free
 }
 {bone1Name=dyn_jacket_bck_l_01, bone2Name=dyn_jacket_bck_l_base, swing1LimitDegrees=15, swing2LimitDegrees=80, twistLimitDegrees=5}
 {bone1Name=dyn_jacket_bck_l_02, bone2Name=dyn_jacket_bck_l_01, swing1LimitDegrees=15, swing2LimitDegrees=80, twistLimitDegrees=5}
 {bone1Name=dyn_jacket_bck_l_03, bone2Name=dyn_jacket_bck_l_02, swing1LimitDegrees=15, swing2LimitDegrees=100, twistLimitDegrees=5}
 {bone1Name=dyn_jacket_bck_l_04, bone2Name=dyn_jacket_bck_l_03, swing1LimitDegrees=15, swing2LimitDegrees=100, twistLimitDegrees=5}
 {bone1Name=dyn_jacket_bck_l_05, bone2Name=dyn_jacket_bck_l_04, swing1LimitDegrees=15, swing2LimitDegrees=100, twistLimitDegrees=5}
 {
  bone1Name=dyn_jacket_bck_r_base
  bone2Name=spine_01
  swing1LimitDegrees=45
  swing1Motion=Free
  swing2LimitDegrees=45
  swing2Motion=Free
  twistLimitDegrees=45
  twistMotion=Free
 }
 {bone1Name=dyn_jacket_bck_r_01, bone2Name=dyn_jacket_bck_r_base, swing1LimitDegrees=15, swing2LimitDegrees=80, twistLimitDegrees=5}
 {bone1Name=dyn_jacket_bck_r_02, bone2Name=dyn_jacket_bck_r_01, swing1LimitDegrees=15, swing2LimitDegrees=80, twistLimitDegrees=5}
 {bone1Name=dyn_jacket_bck_r_03, bone2Name=dyn_jacket_bck_r_02, swing1LimitDegrees=15, swing2LimitDegrees=100, twistLimitDegrees=5}
 {bone1Name=dyn_jacket_bck_r_04, bone2Name=dyn_jacket_bck_r_03, swing1LimitDegrees=15, swing2LimitDegrees=100, twistLimitDegrees=5}
 {bone1Name=dyn_jacket_bck_r_05, bone2Name=dyn_jacket_bck_r_04, swing1LimitDegrees=15, swing2LimitDegrees=100, twistLimitDegrees=5}
 {
  bone1Name=dyn_jacket_fr_l_base
  bone2Name=spine_01
  swing1LimitDegrees=45
  swing1Motion=Free
  swing2LimitDegrees=45
  swing2Motion=Free
  twistLimitDegrees=45
  twistMotion=Free
 }
 {bone1Name=dyn_jacket_fr_l_01, bone2Name=dyn_jacket_fr_l_base, swing1LimitDegrees=15, swing2LimitDegrees=80, twistLimitDegrees=5}
 {bone1Name=dyn_jacket_fr_l_02, bone2Name=dyn_jacket_fr_l_01, swing1LimitDegrees=15, swing2LimitDegrees=80, twistLimitDegrees=5}
 {bone1Name=dyn_jacket_fr_l_03, bone2Name=dyn_jacket_fr_l_02, swing1LimitDegrees=30, swing2LimitDegrees=100, twistLimitDegrees=5}
 {bone1Name=dyn_jacket_fr_l_04, bone2Name=dyn_jacket_fr_l_03, swing1LimitDegrees=30, swing2LimitDegrees=100, twistLimitDegrees=15}
 {bone1Name=dyn_jacket_fr_l_05, bone2Name=dyn_jacket_fr_l_04, swing1LimitDegrees=30, swing2LimitDegrees=100, twistLimitDegrees=15}
 {
  bone1Name=dyn_jacket_fr_r_base
  bone2Name=spine_01
  swing1LimitDegrees=45
  swing1Motion=Free
  swing2LimitDegrees=45
  swing2Motion=Free
  twistLimitDegrees=45
  twistMotion=Free
 }
 {bone1Name=dyn_jacket_fr_r_01, bone2Name=dyn_jacket_fr_r_base, swing1LimitDegrees=15, swing2LimitDegrees=80, twistLimitDegrees=5}
 {bone1Name=dyn_jacket_fr_r_02, bone2Name=dyn_jacket_fr_r_01, swing1LimitDegrees=15, swing2LimitDegrees=80, twistLimitDegrees=5}
 {bone1Name=dyn_jacket_fr_r_03, bone2Name=dyn_jacket_fr_r_02, swing1LimitDegrees=15, swing2LimitDegrees=100, twistLimitDegrees=5}
 {bone1Name=dyn_jacket_fr_r_04, bone2Name=dyn_jacket_fr_r_03, swing1LimitDegrees=15, swing2LimitDegrees=100, twistLimitDegrees=5}
 {bone1Name=dyn_jacket_fr_r_05, bone2Name=dyn_jacket_fr_r_04, swing1LimitDegrees=15, swing2LimitDegrees=100, twistLimitDegrees=5}
 {
  bone1Name=dyn_jacket_side_l_base
  bone2Name=spine_01
  swing1LimitDegrees=45
  swing1Motion=Free
  swing2LimitDegrees=45
  swing2Motion=Free
  twistLimitDegrees=45
  twistMotion=Free
 }
 {bone1Name=dyn_jacket_side_l_01, bone2Name=dyn_jacket_side_l_base, swing1LimitDegrees=15, swing2LimitDegrees=80, twistLimitDegrees=5}
 {bone1Name=dyn_jacket_side_l_02, bone2Name=dyn_jacket_side_l_01, swing1LimitDegrees=15, swing2LimitDegrees=80, twistLimitDegrees=5}
 {bone1Name=dyn_jacket_side_l_03, bone2Name=dyn_jacket_side_l_02, swing1LimitDegrees=15, swing2LimitDegrees=100, twistLimitDegrees=5}
 {bone1Name=dyn_jacket_side_l_04, bone2Name=dyn_jacket_side_l_03, swing1LimitDegrees=15, swing2LimitDegrees=100, twistLimitDegrees=5}
 {bone1Name=dyn_jacket_side_l_05, bone2Name=dyn_jacket_side_l_04, swing1LimitDegrees=15, swing2LimitDegrees=100, twistLimitDegrees=5}
 {
  bone1Name=dyn_jacket_side_r_base
  bone2Name=spine_01
  swing1LimitDegrees=45
  swing1Motion=Free
  swing2LimitDegrees=45
  swing2Motion=Free
  twistLimitDegrees=45
  twistMotion=Free
 }
 {bone1Name=dyn_jacket_side_r_01, bone2Name=dyn_jacket_side_r_base, swing1LimitDegrees=15, swing2LimitDegrees=80, twistLimitDegrees=5}
 {bone1Name=dyn_jacket_side_r_02, bone2Name=dyn_jacket_side_r_01, swing1LimitDegrees=15, swing2LimitDegrees=80, twistLimitDegrees=5}
 {bone1Name=dyn_jacket_side_r_03, bone2Name=dyn_jacket_side_r_02, swing1LimitDegrees=15, swing2LimitDegrees=100, twistLimitDegrees=5}
 {bone1Name=dyn_jacket_side_r_04, bone2Name=dyn_jacket_side_r_03, swing1LimitDegrees=15, swing2LimitDegrees=100, twistLimitDegrees=5}
 {
  bone1Name=spine_01
  bone2Name=pelvis
  swing1LimitDegrees=45
  swing1Motion=Free
  swing2LimitDegrees=45
  swing2Motion=Free
  twistLimitDegrees=45
  twistMotion=Free
 }
 {
  bone1Name=dyn_jacket_fr_r_03
  bone2Name=dyn_jacket_side_r_03
  swing1LimitDegrees=45
  swing1Motion=Free
  swing2LimitDegrees=45
  swing2Motion=Free
  twistLimitDegrees=45
  twistMotion=Free
 }
 {
  bone1Name=dyn_jacket_fr_l_03
  bone2Name=dyn_jacket_side_l_03
  swing1LimitDegrees=45
  swing1Motion=Free
  swing2LimitDegrees=45
  swing2Motion=Free
  twistLimitDegrees=45
  twistMotion=Free
 }
 {
  bone1Name=dyn_jacket_fr_l_05
  bone2Name=dyn_jacket_side_l_05
  swing1LimitDegrees=45
  swing1Motion=Free
  swing2LimitDegrees=45
  swing2Motion=Free
  twistLimitDegrees=45
  twistMotion=Free
 }
 {
  bone1Name=dyn_jacket_fr_l_04
  bone2Name=dyn_jacket_side_l_04
  swing1LimitDegrees=45
  swing1Motion=Free
  swing2LimitDegrees=45
  swing2Motion=Free
  twistLimitDegrees=45
  twistMotion=Free
 }
 {
  bone1Name=dyn_jacket_fr_r_05
  bone2Name=dyn_jacket_side_r_05
  swing1LimitDegrees=45
  swing1Motion=Free
  swing2LimitDegrees=45
  swing2Motion=Free
  twistLimitDegrees=45
  twistMotion=Free
 }
 {
  bone1Name=dyn_jacket_fr_r_04
  bone2Name=dyn_jacket_side_r_04
  swing1LimitDegrees=45
  swing1Motion=Free
  swing2LimitDegrees=45
  swing2Motion=Free
  twistLimitDegrees=45
  twistMotion=Free
 }
]
