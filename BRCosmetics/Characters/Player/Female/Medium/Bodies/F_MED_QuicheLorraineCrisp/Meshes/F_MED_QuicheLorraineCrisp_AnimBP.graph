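# Animation Blueprint that adds secondary motion to the QuicheLorraineCrisp female character body, copying the attached parent's pose and curves.
# A Control Rig takes emote and front-end inputs, then feeds rigid-body nodes using standard and skydive physics assets.
# Anim Dynamics springs necklaces, while bone offsets and constraints adjust necklaces and jacket tails.
# Events set emote, front-end and skydive blend values, simulation settings, and backbling or cape flags.
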AnimBlueprint /BRCosmetics/Characters/Player/Female/Medium/Bodies/F_MED_QuicheLorraineCrisp/Meshes/F_MED_QuicheLorraineCrisp_AnimBP
target_skeleton=&/F_MED_QuicheLorraineCrisp_Skeleton
graphs [
 AnimGraph {nodes [
   n1 = Control Rig!(
    Source=Copy Pose From Mesh() {bCopyCurves=true, bUseAttachedParent=true}
    Alpha
    CustomEmoteLime:double=Get CustomEmoteLime() -> (CustomEmoteLime:double)
    CustomEmoteCrisp:double=Get CustomEmoteCrisp() -> (CustomEmoteCrisp:double)
    FrontEnd Alpha:double=Get FrontEnd Alpha() -> (FrontEnd Alpha:double)
    Emote Alpha:double=Get Emote Alpha() -> (Emote Alpha:double)
   ) -> (Pose) {controlRigAssetReference {blueprintRigClass=&/F_MED_QuicheLorraineCrisp_CtrlRig.F_MED_QuicheLorraineCrisp_CtrlRig_C}}
   n2 = RigidBody!(
    ComponentPose=Local To Component(LocalPose=@n1)
    Alpha
    OverrideWorldGravity:Vector="0,0,-980"
    ExternalForce
    SimSpaceSettings:Sim Space Settings Structure=(WorldAlpha=1,VelocityScaleZ=1,DampingAlpha=0,MaxLinearVelocity=700,MaxAngularVelocity=8,MaxLinearAcceleration=1500,MaxAngularAcceleration=40,ExternalLinearDragV=(X=0.1,Y=0.1,Z=0),ExternalLinearVelocity=(X=0,Y=0,Z=0),ExternalAngularVelocity=(X=0,Y=0,Z=0))
   ) -> (Pose) {
    bClampLinearTranslationLimitToRefPose=true
    bOverrideWorldGravity=true
    overridePhysicsAsset=&/F_MED_QuicheLorraineCrisp_Physics
    simulationSpace=BaseBoneSpace
    baseBoneRef {boneName=spine_01}
    componentAppliedLinearAccClamp=<3000 3000 3000>
    overrideWorldGravity {z=-980}
    simSpaceSettings {
     dampingAlpha=0
     maxAngularAcceleration=40
     maxAngularVelocity=8
     maxLinearAcceleration=1500
     maxLinearVelocity=700
     worldAlpha=1
     externalLinearDragV=<0.1 0.1>
    }
   }
   n3 = RigidBody(
    ComponentPose=@n2
    OverrideWorldGravity="0,0,-980"
    SimSpaceSettings=(WorldAlpha=1,VelocityScaleZ=1,DampingAlpha=0,MaxLinearVelocity=700,MaxAngularVelocity=8,MaxLinearAcceleration=1500,MaxAngularAcceleration=40,ExternalLinearDragV=(X=0.1,Y=0.1,Z=0),ExternalLinearVelocity=(X=0,Y=0,Z=0),ExternalAngularVelocity=(X=0,Y=0,Z=0))
   ) {
    bClampLinearTranslationLimitToRefPose=true
    bOverrideWorldGravity=true
    overridePhysicsAsset=&/F_MED_QuicheLorraineCrisp_Skydive_Physics
    simulationSpace=BaseBoneSpace
    baseBoneRef {boneName=spine_01}
    componentAppliedLinearAccClamp=<3000 3000 3000>
    overrideWorldGravity {z=-980}
    simSpaceSettings {
     dampingAlpha=0
     maxAngularAcceleration=40
     maxAngularVelocity=8
     maxLinearAcceleration=1500
     maxLinearVelocity=700
     worldAlpha=1
     externalLinearDragV=<0.1 0.1>
    }
   }
   n4 = "Transform (Modify) Bone - Bone: dyn_necklaces_base"(ComponentPose=@n3, Alpha=1, Translation="-0.095,-0.655,0.852", Rotation="0.000,0.032,0.023", Scale="1,1,1") {
    rotationMode=BMM_Additive
    rotationSpace=BCS_BoneSpace
    scaleMode=BMM_Replace
    scaleSpace=BCS_BoneSpace
    translationMode=BMM_Additive
    translationSpace=BCS_BoneSpace
    boneToModify {boneName=dyn_necklaces_base}
    rotation=<rot 0 0.032 0.023>
    translation=<-0.095 -0.655 0.8519>
   }
   n5 = Anim Dynamics - Bone: dyn_necklace_A(ComponentPose=@n4, Alpha=1, ExternalForce="0,0,0") {
    angularBiasOverride=0.9
    angularDampingOverride=0.999
    angularSpringConstant=200
    bAngularSpring=true
    bGravityOverrideInSimSpace=true
    bOverrideAngularBias=true
    bOverrideAngularDamping=true
    bUseGravityOverride=true
    simulationSpace=RootRelative
    alphaBoolBlend {blendOption=Linear}
    boundBone {boneName=dyn_necklace_A}
    componentAppliedLinearAccClamp=<1000 1000 1000>
    componentLinearAccScale=<1 1 1>
    componentLinearVelScale=<1 1 1>
    gravityOverride {z=-980}
    physicsBodyDefinitions [{
      boundBone {boneName=dyn_necklace_A}
      boxExtents=<8 8 8>
      constraintSetup {angularLimitsMax=<0 0 2>, angularLimitsMin=<0 -50 -2>, angularTarget=<1 0 0>}
      localJointOffset=<0.5 0 0>
     }]
    simSpaceSettings {simSpaceAngularAlpha=1}
   }
   n6 = Anim Dynamics - Bone: dyn_necklace_B_01(ComponentPose=@n5, Alpha=1, ExternalForce="0,0,0") {
    angularBiasOverride=0.9
    angularDampingOverride=0.999
    angularSpringConstant=200
    bAngularSpring=true
    bGravityOverrideInSimSpace=true
    bOverrideAngularBias=true
    bOverrideAngularDamping=true
    bUseGravityOverride=true
    simulationSpace=RootRelative
    alphaBoolBlend {blendOption=Linear}
    boundBone {boneName=dyn_necklace_B_01}
    componentAppliedLinearAccClamp=<1000 1000 1000>
    componentLinearAccScale=<1 1 1>
    componentLinearVelScale=<1 1 1>
    gravityOverride {z=-980}
    physicsBodyDefinitions [{
      boundBone {boneName=dyn_necklace_B_01}
      boxExtents=<10 10 10>
      constraintSetup {angularLimitsMax=<0 0 2>, angularLimitsMin=<0 -30 -2>, angularTarget=<1 0 0>}
      localJointOffset=<0.5 0 0>
     }]
    simSpaceSettings {simSpaceAngularAlpha=1}
   }
   n7 = Anim Dynamics - Bone: dyn_necklace_B_02(ComponentPose=@n6, Alpha=1, ExternalForce="0,0,0") {
    angularBiasOverride=0.9
    angularDampingOverride=0.999
    angularSpringConstant=200
    bAngularSpring=true
    bGravityOverrideInSimSpace=true
    bOverrideAngularBias=true
    bOverrideAngularDamping=true
    bUseGravityOverride=true
    simulationSpace=RootRelative
    alphaBoolBlend {blendOption=Linear}
    boundBone {boneName=dyn_necklace_B_02}
    componentAppliedLinearAccClamp=<1000 1000 1000>
    componentLinearAccScale=<1 1 1>
    componentLinearVelScale=<1 1 1>
    gravityOverride {z=-980}
    physicsBodyDefinitions [{
      boundBone {boneName=dyn_necklace_B_02}
      boxExtents=<10 10 10>
      constraintSetup {angularLimitsMax=<0 0 2>, angularLimitsMin=<0 -30 -2>, angularTarget=<1 0 0>}
      localJointOffset=<0.5 0 0>
     }]
    simSpaceSettings {simSpaceAngularAlpha=1}
   }
   n8 = Anim Dynamics - Bone: dyn_necklace_C_01(ComponentPose=@n7, Alpha=1, ExternalForce="0,0,0") {
    angularBiasOverride=0.9
    angularDampingOverride=0.999
    angularSpringConstant=200
    bAngularSpring=true
    bGravityOverrideInSimSpace=true
    bOverrideAngularBias=true
    bOverrideAngularDamping=true
    bUseGravityOverride=true
    simulationSpace=RootRelative
    alphaBoolBlend {blendOption=Linear}
    boundBone {boneName=dyn_necklace_C_01}
    componentAppliedLinearAccClamp=<1000 1000 1000>
    componentLinearAccScale=<1 1 1>
    componentLinearVelScale=<1 1 1>
    gravityOverride {z=-980}
    physicsBodyDefinitions [{
      boundBone {boneName=dyn_necklace_C_01}
      boxExtents=<10 10 10>
      constraintSetup {angularLimitsMax=<0 0 2>, angularLimitsMin=<0 -20 -2>, angularTarget=<1 0 0>}
      localJointOffset=<0.5 0 0>
     }]
    simSpaceSettings {simSpaceAngularAlpha=1}
   }
   n9 = Anim Dynamics - Bone: dyn_necklace_C_02(ComponentPose=@n8, Alpha=1, ExternalForce="0,0,0") {
    angularBiasOverride=0.9
    angularDampingOverride=0.999
    angularSpringConstant=200
    bAngularSpring=true
    bGravityOverrideInSimSpace=true
    bOverrideAngularBias=true
    bOverrideAngularDamping=true
    bUseGravityOverride=true
    simulationSpace=RootRelative
    alphaBoolBlend {blendOption=Linear}
    boundBone {boneName=dyn_necklace_C_02}
    componentAppliedLinearAccClamp=<1000 1000 1000>
    componentLinearAccScale=<1 1 1>
    componentLinearVelScale=<1 1 1>
    gravityOverride {z=-980}
    physicsBodyDefinitions [{
      boundBone {boneName=dyn_necklace_C_02}
      boxExtents=<10 10 10>
      constraintSetup {angularLimitsMax=<0 0 2>, angularLimitsMin=<0 -20 -2>, angularTarget=<1 0 0>}
      localJointOffset=<0.5 0 0>
     }]
    simSpaceSettings {simSpaceAngularAlpha=1}
   }
   n10 = "Transform (Modify) Bone - Bone: dyn_jacket_bck_c_02"(ComponentPose=@n9, Alpha=1, Translation="0,0,0", Rotation="30.000,-0.024,0.035", Scale="1,1,1") {
    rotationMode=BMM_Additive
    rotationSpace=BCS_BoneSpace
    translationMode=BMM_Additive
    translationSpace=BCS_BoneSpace
    boneToModify {boneName=dyn_jacket_bck_c_02}
    rotation=<rot 30 -0.024 0.035>
   }
   n11 = "Transform (Modify) Bone - Bone: dyn_jacket_bck_r_02"(ComponentPose=@n10, Alpha=1, Translation="2.941,6.192,9.229", Rotation="29.998,0.162,0.005", Scale="1,1,1") {
    rotationMode=BMM_Additive
    rotationSpace=BCS_BoneSpace
    translationMode=BMM_Additive
    translationSpace=BCS_BoneSpace
    boneToModify {boneName=dyn_jacket_bck_r_02}
    rotation=<rot 29.998 0.162 0.005>
    translation=<2.941 6.192 9.229>
   }
   n12 = "Transform (Modify) Bone - Bone: dyn_jacket_bck_l_02"(ComponentPose=@n11, Alpha=1, Translation="-0.057,-11.798,-2.724", Rotation="15.000,0.031,-0.006", Scale="1,1,1") {
    rotationMode=BMM_Additive
    rotationSpace=BCS_BoneSpace
    translationMode=BMM_Additive
    translationSpace=BCS_BoneSpace
    boneToModify {boneName=dyn_jacket_bck_l_02}
    rotation=<rot 15 0.031 -0.006>
    translation=<-0.057 -11.798 -2.724>
   }
   n13 = "Transform (Modify) Bone - Bone: dyn_jacket_bck_c_03"(ComponentPose=@n12, Alpha=1, Translation="0,0,0", Rotation="24.594,10.986,4.636", Scale="1,1,1") {
    rotationMode=BMM_Additive
    rotationSpace=BCS_BoneSpace
    translationMode=BMM_Additive
    translationSpace=BCS_BoneSpace
    boneToModify {boneName=dyn_jacket_bck_c_03}
    rotation=<rot 24.594 10.986 4.636>
   }
   n14 = "Transform (Modify) Bone - Bone: dyn_jacket_bck_l_02"(ComponentPose=@n13, Alpha=1, Translation="0,0,0", Rotation="13.515,-20.269,-8.549", Scale="1,1,1") {
    rotationMode=BMM_Additive
    rotationSpace=BCS_BoneSpace
    translationMode=BMM_Additive
    translationSpace=BCS_BoneSpace
    boneToModify {boneName=dyn_jacket_bck_l_02}
    rotation=<rot 13.515 -20.269 -8.549>
   }
   n15 = "Transform (Modify) Bone - Bone: dyn_jacket_bck_l_03"(ComponentPose=@n14, Alpha=1, Translation="0,0,0", Rotation="5.000,0.000,0.000", Scale="1,1,1") {
    rotationMode=BMM_Additive
    rotationSpace=BCS_BoneSpace
    translationMode=BMM_Additive
    translationSpace=BCS_BoneSpace
    boneToModify {boneName=dyn_jacket_bck_l_03}
    rotation {pitch=5}
   }
   n16 = "Constraint - dyn_necklaces_base (PPPP)"!(
    ComponentPose=@n15
    Alpha=1
    ConstraintWeights_0=0.9857
    ConstraintWeights_1:float=0.0107
    ConstraintWeights_2:float=0.002
    ConstraintWeights_3:float=0.0016
   ) -> (Pose) {
    boneToModify {boneName=dyn_necklaces_base}
    constraintSetup [{targetBone {boneName=spine_05}}, {targetBone {boneName=neck_01}}, {targetBone {boneName=clavicle_r}}, {targetBone {boneName=spine_04}}]
    constraintWeights [1,1,1,1]
   }
   n17 = "Constraint - dyn_jacket_bck_l_01 (P)"(ComponentPose=@n16, Alpha=1, ConstraintWeights_0=1) {
    alphaScaleBiasClamp {scale=0.18}
    boneToModify {boneName=dyn_jacket_bck_l_01}
    constraintSetup [{targetBone {boneName=thigh_l}}]
    constraintWeights [1]
   }
   n18 = "Constraint - dyn_jacket_bck_c_01 (PP)"!(ComponentPose=@n17, Alpha=1, ConstraintWeights_0=0.5, ConstraintWeights_1:float=0.5) -> (Pose) {
    alphaScaleBiasClamp {scale=0.18}
    boneToModify {boneName=dyn_jacket_bck_c_01}
    constraintSetup [{targetBone {boneName=thigh_l}}, {targetBone {boneName=thigh_r}}]
    constraintWeights [1,1]
   }
   n19 = "Constraint - dyn_jacket_bck_r_01 (T)"(ComponentPose=@n18, Alpha=1, ConstraintWeights_0=1) {
    alphaScaleBiasClamp {scale=0.18}
    boneToModify {boneName=dyn_jacket_bck_r_01}
    constraintSetup [{transformType=Translation, targetBone {boneName=thigh_r}}]
    constraintWeights [1]
   }
   Output Pose(Result=Component To Local(ComponentPose=@n19)) {layerGroup=None}
  ]}
 EventGraph {nodes [
   Event OnCharacterCustomizationLoaded()
   > n1 = Branch(Condition=Get bIsInFrontEnd())
   > Set FrontEnd Alpha(FrontEnd Alpha:double=0) -> (Output_Get:double)
   Set FrontEnd Alpha(execute=@n1.else, FrontEnd Alpha=1)
   Event OnBeginSkydiving()
   > Set Skydive Alpha(Skydive Alpha:double=1) -> (Output_Get:double)
   > Set Normal Alpha(Normal Alpha:double=0) -> (Output_Get:double)
   Event OnEndSkydiving()
   > Set Skydive Alpha(Skydive Alpha=0)
   > Set Normal Alpha(Normal Alpha=1)
   n2 = Event BlueprintUpdateAnimation()
   > n3 = Sequence()
   > n4 = "Gravity Override State Function (Parameter Asset) (Nativized)"(
    CharacterPartAnimInstance=Self-Reference()
    GameTimeInSeconds=@n2.DeltaTimeX
    DebugJointName=neck_02
    Parameters=&/F_MED_QuicheLorraineCrisp_Parameters
   )
   > Set Gravity Override(Gravity Override=@n4)
   n5 = Break Fort Rigid Body Anim Node Parameters!(FortRigidBodyAnimNodeParameters:"Fort Rigid Body Anim Node Parameters Structure (by ref)"=@n4.OutRigidBodyAnimNodeParameters) -> (ComponentLinearAccScale:Vector=~, ComponentLinearVelScale:Vector=~, SimSpaceSettings:Sim Space Settings Structure) {is=title K2Node_BreakStruct}
   > Set Sim Space Settings(Sim Space Settings:Sim Space Settings Structure=@n5.SimSpaceSettings) -> (Output_Get:Sim Space Settings Structure)
   > Set CustomEmoteLime(CustomEmoteLime:double=GetCurveValue(CurveName=QuicheLorraineLimeEmote)) -> (Output_Get:double)
   n6 = GetCurveValue(CurveName=QuicheLorraineCrispEmote)
   > Set CustomEmoteCrisp(CustomEmoteCrisp:double=@n6) -> (Output_Get:double)
   > n7 = Branch(Condition="float > float"(A=@n6))
   > n8 = Set Sim Space Settings()
   Set CrispEmoteTweak(
    execute=[@n9.then, @n8.then]
    CrispEmoteTweak:double=float * float(A=Get CustomEmoteCrisp(), B=GetCurveValue(CurveName=DeltaRef_Calf_L)) {is=title K2Node_PromotableOperator}
   ) -> (Output_Get:double)
   n9 = Set Sim Space Settings(execute=@n7.else, Sim Space Settings=@n5.SimSpaceSettings)
   n10 = CheckForBackbling()
   Set isWearingBackbling(
    execute=@n3.then_1
    isWearingBackbling=AND Boolean!(A=@n10, B=NOT Boolean(A=@n10.bIsBackblingACape), C:bool=NOT Boolean(A=@n10.bIsBackblingLong)) -> (ReturnValue)
   )
   > Set Is Wearing Cape(Is Wearing Cape=OR Boolean(A=@n10.bIsBackblingACape, B=@n10.bIsBackblingLong))
   > Branch(Condition=AND Boolean!(A=Get Is Wearing Cape() -> (Is Wearing Cape:bool), B=Get bIsSkyDiving()) -> (ReturnValue))
   > Set Skydive Alpha(Skydive Alpha=0.5)
   Event OnBeginPlayingEmote() {is=title K2Node_Event}
   > n11 = Branch(Condition=Get bIsInFrontEnd())
   > n12 = Set FrontEnd Alpha(FrontEnd Alpha=1)
   Set Emote Alpha(execute=[@n13.then, @n12.then], Emote Alpha:double=1) -> (Output_Get:double)
   n13 = Set FrontEnd Alpha(execute=@n11.else, FrontEnd Alpha=1)
   Event OnEndPlayingEmote() {is=title K2Node_Event}
   > n14 = Branch(Condition=Get bIsInFrontEnd())
   > n15 = Set FrontEnd Alpha(FrontEnd Alpha=0)
   Set Emote Alpha(execute=[@n16.then, @n15.then], Emote Alpha=0)
   n16 = Set FrontEnd Alpha(execute=@n14.else, FrontEnd Alpha=1)
  ]}
]
variables [
 CrispEmoteTweak {type=number}
 CustomEmoteCrisp {type=number}
 CustomEmoteLime {type=number}
 Emote Alpha {type=number}
 FrontEnd Alpha {type=number}
 Gravity Override {type=Vector}
 Is Wearing Cape {type=boolean}
 isWearingBackbling {type=boolean}
 L calfBend Alpha {type=number}
 Normal Alpha {type=number, default=1}
 Sim Space Settings {
  type=SimSpaceSettings
  default {
   dampingAlpha=0
   maxAngularAcceleration=40
   maxAngularVelocity=10
   maxLinearAcceleration=1200
   maxLinearVelocity=600
   velocityScaleZ=0.5
   worldAlpha=1
   externalLinearDragV=<0.3 0.3 0.3>
  }
 }
 Skydive Alpha {type=number}
]
