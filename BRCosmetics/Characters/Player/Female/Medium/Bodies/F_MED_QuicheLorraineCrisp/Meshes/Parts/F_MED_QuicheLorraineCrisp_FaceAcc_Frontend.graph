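# Physics asset for the Frontend FaceAcc mesh of the female medium QuicheLorraineCrisp body cosmetic, centered on a front hair bone chain.
# Capsules cover the hair chain plus head, neck, torso, coat lapels, clavicles, thighs and the right arm only.
# Hair segments 02–07 chain from the shoulder base with swing and twist limits.
# Upper links from head through head base and hair 01 to shoulder base, and listed body and lapel constraints, leave angular motion free.
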PhysicsAsset /BRCosmetics/Characters/Player/Female/Medium/Bodies/F_MED_QuicheLorraineCrisp/Meshes/Parts/F_MED_QuicheLorraineCrisp_FaceAcc_Frontend
bodies {
 clavicle_l [{shapeType=Capsule, length=5, radius=5, shapeName=clavicle_l_capsule, center=<3.7682 0 0>, rotation=<rot -90 -0.0002 0>}]
 clavicle_r [{shapeType=Capsule, length=5, radius=5, shapeName=clavicle_r_capsule, center=<-3.7682 0 0>, rotation=<rot 90 0.0002 0>}]
 deform_coat_lapel_l_bot [{shapeType=Capsule, length=10, radius=5, shapeName=deform_coat_lapel_l_bot_capsule, center=<1.1659 -1.7662 0>}]
 deform_coat_lapel_l_mid [{
   shapeType=Capsule
   length=10
   radius=5
   shapeName=deform_coat_lapel_l_mid_capsule
   center=<1.8181 -0.6076 0.025>
   rotation=<rot 20.0001 0 10>
  }]
 deform_coat_lapel_r_bot [{shapeType=Capsule, length=11.5, radius=3.5, shapeName=deform_coat_lapel_r_bot_capsule, center=<3.167 1.9314 -3.3507>}]
 deform_coat_lapel_r_mid [{shapeType=Capsule, length=11, radius=4, shapeName=deform_coat_lapel_r_mid_capsule, center=<3.7177 0.0253 -3.3507>}]
 dyn_hairFront_01 [{shapeType=Capsule, length=4.5, radius=2.5, shapeName=dyn_hairFront_01_capsule, center=<3.8572 0 0>, rotation=<rot 90 0 0>}]
 dyn_hairFront_02 [{shapeType=Capsule, length=5.5, radius=2.5, shapeName=dyn_hairFront_02_capsule, center=<1.8 0 0.5>, rotation=<rot 79.9999 180 180>}]
 dyn_hairFront_03 [{shapeType=Capsule, length=5.5, radius=2.5, shapeName=dyn_hairFront_03_capsule, center=<3.0604 0 1.0791>, rotation=<rot 90 0 0>}]
 dyn_hairFront_04 [{shapeType=Capsule, length=4.5, radius=3.5, shapeName=dyn_hairFront_04_capsule, center=<3.8572 0 0.9647>, rotation=<rot 90 0 0>}]
 dyn_hairFront_05 [{shapeType=Capsule, length=5, radius=3, shapeName=dyn_hairFront_05_capsule, center=<3.8572 0 0.9647>, rotation=<rot 90 0 0>}]
 dyn_hairFront_06 [{shapeType=Capsule, length=5, radius=3, shapeName=dyn_hairFront_06_capsule, center=<3.8572 0 0.9647>, rotation=<rot 90 0 0>}]
 dyn_hairFront_07 [{shapeType=Capsule, length=5, radius=3, shapeName=dyn_hairFront_07_capsule, center=<3.8572 0 0.9647>, rotation=<rot 90 0 0>}]
 dyn_hairFront_head_base [{shapeType=Capsule, length=5, radius=3, shapeName=dyn_hairFront_head_base_capsule, rotation=<rot 0 -90 0>}]
 dyn_hairFront_shoulder_base [{shapeType=Capsule, length=0, radius=0.5, shapeName=dyn_hairFront_shoulder_base_capsule, rotation=<rot 0 -90 -90.0001>}]
 head [{shapeType=Capsule, length=6.5, radius=5.5, shapeName=head_capsule, rotation=<rot -90 0 0>}]
 lowerarm_r [{shapeType=Capsule, length=20, radius=5, shapeName=lowerarm_r_capsule, center=<-15.9894 0 3.7706>, rotation=<rot 90 0.0001 0>}]
 neck_01 [{shapeType=Capsule, length=59.5, radius=13.5, shapeName=neck_01_capsule, center=<0 -8.6815 0>}]
 neck_02 [{shapeType=Capsule, length=5, radius=3, shapeName=neck_02_capsule}]
 pelvis [{shapeType=Capsule, length=11.5, radius=11.5, shapeName=pelvis_capsule}]
 spine_01 [{shapeType=Capsule, length=19.5, radius=10, shapeName=spine_01_capsule}]
 spine_03 [{shapeType=Capsule, length=21.5, radius=10, shapeName=spine_03_capsule, center=<-2.7436 0.5372 0>}]
 spine_04 [{shapeType=Capsule, length=10.5, radius=12.5, shapeName=spine_04_capsule, center=<0 0.96 0>}]
 spine_05 [{shapeType=Capsule, length=23.5, radius=9.5, shapeName=spine_05_capsule, center=<0.5269 -0.5262 0>}]
 thigh_l [{shapeType=Capsule, length=43, radius=13, shapeName=thigh_l_capsule, center=<-26.5987 0.0001 0>, rotation=<rot -90 -0.0002 0>}]
 thigh_r [{shapeType=Capsule, length=43, radius=13, shapeName=thigh_r_capsule, center=<26.5987 -0.0002 0>, rotation=<rot 90 -0.0003 0>}]
 upperarm_r [{shapeType=Capsule, length=20.5, radius=7.5, shapeName=upperarm_r_capsule, center=<-13.0603 0.3984 3.0193>, rotation=<rot -90 0 0>}]
}
constraints [
 {bone1Name=dyn_hairFront_02, bone2Name=dyn_hairFront_shoulder_base, swing1LimitDegrees=60, swing2LimitDegrees=60, twistLimitDegrees=10}
 {bone1Name=dyn_hairFront_03, bone2Name=dyn_hairFront_02, swing1LimitDegrees=60, swing2LimitDegrees=60, twistLimitDegrees=10}
 {bone1Name=dyn_hairFront_04, bone2Name=dyn_hairFront_03, swing1LimitDegrees=60, swing2LimitDegrees=60, twistLimitDegrees=10}
 {bone1Name=dyn_hairFront_05, bone2Name=dyn_hairFront_04, swing1LimitDegrees=60, swing2LimitDegrees=60, twistLimitDegrees=10}
 {bone1Name=dyn_hairFront_06, bone2Name=dyn_hairFront_05, swing1LimitDegrees=60, swing2LimitDegrees=60, twistLimitDegrees=10}
 {bone1Name=dyn_hairFront_07, bone2Name=dyn_hairFront_06, swing1LimitDegrees=60, swing2LimitDegrees=60, twistLimitDegrees=10}
 {
  bone1Name=spine_01
  bone2Name=pelvis
  swing1LimitDegrees=45
  swing1Motion=Free
  swing2LimitDegrees=45
  swing2Motion=Free
  twistLimitDegrees=45
  twistMotion=Free
 }
 {
  bone1Name=deform_coat_lapel_r_mid
  bone2Name=spine_04
  swing1LimitDegrees=45
  swing1Motion=Free
  swing2LimitDegrees=45
  swing2Motion=Free
  twistLimitDegrees=45
  twistMotion=Free
 }
 {
  bone1Name=deform_coat_lapel_r_bot
  bone2Name=spine_03
  swing1LimitDegrees=45
  swing1Motion=Free
  swing2LimitDegrees=45
  swing2Motion=Free
  twistLimitDegrees=45
  twistMotion=Free
 }
 {
  bone1Name=deform_coat_lapel_l_bot
  bone2Name=spine_03
  swing1LimitDegrees=45
  swing1Motion=Free
  swing2LimitDegrees=45
  swing2Motion=Free
  twistLimitDegrees=45
  twistMotion=Free
 }
 {
  bone1Name=deform_coat_lapel_l_mid
  bone2Name=spine_04
  swing1LimitDegrees=45
  swing1Motion=Free
  swing2LimitDegrees=45
  swing2Motion=Free
  twistLimitDegrees=45
  twistMotion=Free
 }
 {
  bone1Name=thigh_l
  bone2Name=pelvis
  swing1LimitDegrees=45
  swing1Motion=Free
  swing2LimitDegrees=45
  swing2Motion=Free
  twistLimitDegrees=45
  twistMotion=Free
 }
 {
  bone1Name=thigh_r
  bone2Name=pelvis
  swing1LimitDegrees=45
  swing1Motion=Free
  swing2LimitDegrees=45
  swing2Motion=Free
  twistLimitDegrees=45
  twistMotion=Free
 }
 {
  bone1Name=dyn_hairFront_shoulder_base
  bone2Name=dyn_hairFront_01
  swing1LimitDegrees=45
  swing1Motion=Free
  swing2LimitDegrees=45
  swing2Motion=Free
  twistLimitDegrees=45
  twistMotion=Free
 }
 {
  bone1Name=dyn_hairFront_01
  bone2Name=dyn_hairFront_head_base
  swing1LimitDegrees=45
  swing1Motion=Free
  swing2LimitDegrees=45
  swing2Motion=Free
  twistLimitDegrees=45
  twistMotion=Free
 }
 {
  bone1Name=dyn_hairFront_head_base
  bone2Name=head
  swing1LimitDegrees=45
  swing1Motion=Free
  swing2LimitDegrees=45
  swing2Motion=Free
  twistLimitDegrees=45
  twistMotion=Free
 }
]
