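# Post-process style AnimBP for the QuicheLorraineCrisp character's FaceAcc part, animating its front hair.
# Body poses are copied and layered, then pass through weighted bone constraints, small bone offsets and two RigidBody simulations in spine_05 space, using gameplay and frontend physics assets.
# The event graph toggles RBAN/FE alphas for frontend and emotes, retunes the frontend sim-space settings, and tracks gravity override, a lime-emote curve and skydiving.
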AnimBlueprint /BRCosmetics/Characters/Player/Female/Medium/Bodies/F_MED_QuicheLorraineCrisp/Meshes/Parts/F_MED_QuicheLorraineCrisp_FaceAcc_AnimBP
target_skeleton=&/F_MED_QuicheLorraineCrisp_FaceAcc_Skeleton
graphs [
 AnimGraph {nodes [
   Component To Local()
   "Constraint - dyn_hairFront_shoulder_base (PPPP)"!(
    ComponentPose
    Alpha=1
    ConstraintWeights_0=0.0021
    ConstraintWeights_1:float=0.3966
    ConstraintWeights_2:float=0.5743
    ConstraintWeights_3:float=0.027
   ) -> (Pose) {
    boneToModify {boneName=dyn_hairFront_shoulder_base}
    constraintSetup [
     {targetBone {boneName=spine_04}}
     {targetBone {boneName=spine_05}}
     {targetBone {boneName=clavicle_r}}
     {targetBone {boneName=upperarm_twist_01_r}}
    ]
    constraintWeights [1,1,1,1]
   }
   Copy Pose From Mesh() {bUseAttachedParent=true}
   n1 = Copy Pose From Mesh(SourceMeshComponent=Get Body Mesh() -> (Body Mesh:Skeletal Mesh Component Object Reference)) {bCopyCurves=true, bUseAttachedParent=true}
   n2 = Layered blend per bone(
    BasePose=Copy Pose From Mesh() {bCopyCurves=true, bUseAttachedParent=true}
    BlendPoses_0=Copy Pose From Mesh(SourceMeshComponent=Get Body Mesh()) {bCopyCurves=true, bUseAttachedParent=true}
   ) {layerSetup [{branchFilters [{boneName=upperarm_l}, {boneName=upperarm_r}]}]}
   n3 = "Constraint - dyn_hairFront_shoulder_base (PPPP)"(
    ComponentPose=Local To Component(LocalPose=Layered blend per bone(BasePose=@n2, BlendPoses_0=@n1) {layerSetup [{branchFilters [{boneName=thigh_r}, {boneName=thigh_l}]}]})
    Alpha=1
    ConstraintWeights_0=0.0515
    ConstraintWeights_1=0.605
    ConstraintWeights_2=0.3329
    ConstraintWeights_3=0.0106
   ) {
    alphaScaleBiasClamp {bInterpResult=true}
    boneToModify {boneName=dyn_hairFront_shoulder_base}
    constraintSetup [
     {targetBone {boneName=spine_04}}
     {targetBone {boneName=spine_05}}
     {targetBone {boneName=clavicle_r}}
     {targetBone {boneName=upperarm_twist_01_r}}
    ]
    constraintWeights [1,1,1,1]
   }
   n4 = "Transform (Modify) Bone - Bone: dyn_hairFront_head_base"(ComponentPose=@n3, Alpha=1, Translation="0,0,0", Rotation="-40.000,0.000,-0.000", Scale="1,1,1") {
    rotationMode=BMM_Additive
    translationMode=BMM_Additive
    alphaScaleBiasClamp {bInterpResult=true}
    boneToModify {boneName=dyn_hairFront_head_base}
    rotation {pitch=-40}
   }
   n5 = RigidBody!(ComponentPose=@n4, Alpha, OverrideWorldGravity:Vector="0,0,-980", ExternalForce) -> (Pose) {
    bClampLinearTranslationLimitToRefPose=true
    bOverrideWorldGravity=true
    overridePhysicsAsset=&/F_MED_QuicheLorraineCrisp_FaceAcc_Physics
    simulationSpace=BaseBoneSpace
    alphaScaleBiasClamp {bInterpResult=true}
    baseBoneRef {boneName=spine_05}
    componentAppliedLinearAccClamp=<3000 3000 3000>
    overrideWorldGravity {z=-980}
    simSpaceSettings {dampingAlpha=0, maxAngularAcceleration=40, maxAngularVelocity=8, maxLinearAcceleration=1500, maxLinearVelocity=700, worldAlpha=0.2}
   }
   n6 = RigidBody!(
    ComponentPose=@n5
    Alpha
    OverrideWorldGravity="0,0,-980"
    ExternalForce
    SimSpaceSettings:Sim Space Settings Structure=Get Sim Space Settings() -> (Sim Space Settings:Sim Space Settings Structure)
   ) -> (Pose) {
    bClampLinearTranslationLimitToRefPose=true
    bOverrideWorldGravity=true
    overridePhysicsAsset=&/F_MED_QuicheLorraineCrisp_FaceAcc_Frontend
    simulationSpace=BaseBoneSpace
    alphaScaleBiasClamp {bInterpResult=true}
    baseBoneRef {boneName=spine_05}
    componentAppliedLinearAccClamp=<3000 3000 3000>
    overrideWorldGravity {z=-980}
    simSpaceSettings {dampingAlpha=0, maxAngularAcceleration=40, maxAngularVelocity=8, maxLinearAcceleration=1500, maxLinearVelocity=700, worldAlpha=0.2}
   }
   n7 = "Constraint - dyn_hairFront_05 (P)"!(ComponentPose=@n6, Alpha=1, ConstraintWeights_0=1, ConstraintWeights_1:float=0) -> (Pose) {boneToModify {boneName=dyn_hairFront_05}, constraintSetup [{targetBone {boneName=spine_04}}], constraintWeights [1,0]}
   n8 = "Constraint - dyn_hairFront_06 (PPPPP)"!(
    ComponentPose=@n7
    Alpha=1
    ConstraintWeights_0=0.3038
    ConstraintWeights_1:float=0.3543
    ConstraintWeights_2:float=0.1938
    ConstraintWeights_3:float=0.1397
    ConstraintWeights_4:float=0.0085
   ) -> (Pose) {
    boneToModify {boneName=dyn_hairFront_06}
    constraintSetup [
     {targetBone {boneName=spine_02}}
     {targetBone {boneName=spine_04}}
     {targetBone {boneName=spine_03}}
     {targetBone {boneName=spine_01}}
     {targetBone {boneName=pelvis}}
    ]
    constraintWeights [1,1,1,1,1]
   }
   n9 = "Constraint - dyn_hairFront_07 (PPPPP)"!(
    ComponentPose=@n8
    Alpha=1
    ConstraintWeights_0=0.3057
    ConstraintWeights_1:float=0.0784
    ConstraintWeights_2:float=0.0102
    ConstraintWeights_3:float=0.1728
    ConstraintWeights_4:float=0.433
   ) -> (Pose) {
    boneToModify {boneName=dyn_hairFront_07}
    constraintSetup [
     {targetBone {boneName=spine_02}}
     {targetBone {boneName=spine_04}}
     {targetBone {boneName=spine_03}}
     {targetBone {boneName=spine_01}}
     {targetBone {boneName=pelvis}}
    ]
    constraintWeights [1,1,1,1,1]
   }
   n10 = "Transform (Modify) Bone - Bone: jnt_support_hair_l"(ComponentPose=@n9, Alpha=1, Translation="0,0.8,0", Rotation="0,0,0", Scale="1,1,1") {translationMode=BMM_Additive, translationSpace=BCS_BoneSpace, boneToModify {boneName=jnt_support_hair_l}, translation {y=0.8}}
   n11 = "Transform (Modify) Bone - Bone: clavicle_l"(ComponentPose=@n10, Alpha=1, Translation="0,0,0.3", Rotation="0,0,0", Scale="1,1,1") {
    rotationMode=BMM_Additive
    rotationSpace=BCS_BoneSpace
    translationMode=BMM_Additive
    translationSpace=BCS_BoneSpace
    boneToModify {boneName=clavicle_l}
    translation {z=0.3}
   }
   Output Pose(Result=Component To Local(ComponentPose=@n11)) {layerGroup=None}
  ]}
 EventGraph {nodes [
   n1 = Event BlueprintUpdateAnimation()
   > n2 = "Gravity Override State Function (Parameter Asset) (Nativized)"(
    CharacterPartAnimInstance=Self-Reference()
    GameTimeInSeconds=@n1.DeltaTimeX
    DebugJointName=neck_02
    Parameters=&/F_MED_QuicheLorraineCrisp_FaceAcc_Parameters
   )
   > Set Gravity Override(Gravity Override=@n2)
   > Set CustomEmoteLime(CustomEmoteLime:double=GetCurveValue(CurveName=QuicheLorraineLimeEmote)) -> (Output_Get:double)
   Event OnCharacterCustomizationLoaded()
   > n3 = Branch(Condition=Get bIsInFrontEnd())
   > Set FE Alpha(FE Alpha:double=1) -> (Output_Get:double)
   > Set RBAN Alpha(RBAN Alpha:double=0) -> (Output_Get:double)
   Set FE Alpha(execute=@n3.else, FE Alpha=0)
   > Set RBAN Alpha(RBAN Alpha=1)
   n4 = Event OnBeginPlayingEmote() {is=title K2Node_Event}
   Set FE Alpha(execute=[@n4.then, @n5.then], FE Alpha=1)
   > Set RBAN Alpha(RBAN Alpha=0)
   > Set Sim Space Settings(Sim Space Settings:Sim Space Settings Structure=Make Sim Space Settings!(
     WorldAlpha:float=0.2
     VelocityScaleZ:float=1
     DampingAlpha:float=0
     MaxLinearVelocity:float=2
     MaxAngularVelocity:float=2
     MaxLinearAcceleration:float=2
     MaxAngularAcceleration:float=2
     ExternalLinearDragV:Vector="0.2,0.2,0.2"
     ExternalLinearVelocity:Vector="0,0,0"
     ExternalAngularVelocity:Vector="0,0,0"
    ) -> (SimSpaceSettings:Sim Space Settings Structure) {is=title K2Node_MakeStruct}) -> (Output_Get:Sim Space Settings Structure)
   Event BlueprintInitializeAnimation()
   > Set Skel_Mesh(Skel_Mesh=GetOwnerSkeletalMeshComponent())
   Event OnEndPlayingEmote() {is=title K2Node_Event}
   > n5 = Branch(Condition=Get bIsInFrontEnd())
   Set FE Alpha(execute=@n5.else, FE Alpha=0)
   > Set RBAN Alpha(RBAN Alpha=1)
   > Set Sim Space Settings(Sim Space Settings=Make Sim Space Settings(
     WorldAlpha=0.2
     VelocityScaleZ=1
     DampingAlpha=700
     MaxLinearVelocity=8
     MaxAngularVelocity=5
     MaxLinearAcceleration=1500
     MaxAngularAcceleration=40
     ExternalLinearDragV="0,0,0"
     ExternalLinearVelocity="0,0,0"
     ExternalAngularVelocity="0,0,0"
    ))
   Event OnBeginSkydiving()
   > Set skydive alpha(skydive alpha:double=1) -> (Output_Get:double)
   Event OnEndSkydiving()
   > Set skydive alpha(skydive alpha=0)
   Get bIsSkydiveFloating()
  ]}
]
variables [
 Body Mesh {type=SkeletalMeshComponent}
 CustomEmoteLime {type=number}
 FE Alpha {type=number}
 Gravity Override {type=Vector, default=<0 0 -980>}
 RBAN Alpha {type=number}
 Sim Space Settings {
  type=SimSpaceSettings
  default {
   dampingAlpha=0
   maxAngularAcceleration=40
   maxAngularVelocity=8
   maxLinearAcceleration=1500
   maxLinearVelocity=700
   velocityScaleZ=1
   worldAlpha=0.2
   externalLinearDragV=<0 0 0>
  }
 }
 Skel_Mesh {type=SkeletalMeshComponent}
 skydive alpha {type=number}
]
